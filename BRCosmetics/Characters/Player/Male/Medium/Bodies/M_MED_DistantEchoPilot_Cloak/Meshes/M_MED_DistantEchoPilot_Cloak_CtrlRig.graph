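# Control Rig for the Distant Echo Pilot cloak body that procedurally corrects jacket, collar and belt buckle bones.
# Front jacket bones follow pelvis and spine constraints in Pickaxe mode, otherwise blending target aiming and weak constraints by the emote value.
# Spherical pose readers on thighs and right upper arm weight additional bone offsets.
# IsCrouchingCtrl selects between buckle and strap correction chains; some pose-reader nodes are left unconnected.
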ControlRigBlueprint /BRCosmetics/Characters/Player/Male/Medium/Bodies/M_MED_DistantEchoPilot_Cloak/Meshes/M_MED_DistantEchoPilot_Cloak_CtrlRig
graphs [
 ContainedGraph {nodes [
   n1 = RigVMFunctionEntryNode()
   SequenceAggregate_2 = &RigVMFunction_Sequence(ExecuteContext=@n1.ExecuteContext)
   SequenceAggregate_2_1 = &RigVMFunction_Sequence(ExecuteContext=@SequenceAggregate_2.B)
   SequenceAggregate_2_2 = &RigVMFunction_Sequence(ExecuteContext=@SequenceAggregate_2_1.B)
   RigVMFunctionReturnNode(A=@SequenceAggregate_2.A, B=@SequenceAggregate_2_1.A, C=@SequenceAggregate_2_2.A, D=@SequenceAggregate_2_2.B)
  ]}
 RigVMFunctionLibrary {links [], nodes []}
 RigVMModel {nodes [
   RigUnit_BeginExecution = &RigUnit_BeginExecution()
   SequenceAggregate_2 = RigVMAggregateNode(ExecuteContext=@RigUnit_BeginExecution.ExecutePin)
   n1 = &RigVMFunction_ControlFlowBranch(ExecuteContext=@SequenceAggregate_2.A, Condition=RigVMVariableNode() {variable=Pickaxe}.Value)
   n2 = RigVMVariableNode() {variable=emote}
   /Script/ControlRig.RigUnit_AimItem(
    DebugSettings=(Scale=10,WorldOffset=(Rotation=(X=0,Y=0,Z=0,W=1),Translation=(X=0,Y=0,Z=0),Scale3D=(X=1,Y=1,Z=1)))
    Item=(Type=Bone,Name="Jacket_Fr_l_lookat")
    Primary=(Weight=1,Axis=(X=1,Y=0,Z=0),Target=(X=1,Y=0,Z=0),Kind=Location,Space=(Type=Bone,Name="dyn_Jacket_fr_TGT"))
    Secondary=(Weight=1,Axis=(X=0,Y=1,Z=0),Target=(X=1,Y=0,Z=0),Kind=Location,Space=(Type=Control,Name="pelvis_aim"))
    ExecutePin=@n1.False
    Weight=@n2.Value
   ) {is=struct RigVMUnitNode}
   > &RigUnit_AimItem(
    DebugSettings=(Scale=10,WorldOffset=(Rotation=(X=0,Y=0,Z=0,W=1),Translation=(X=0,Y=0,Z=0),Scale3D=(X=1,Y=1,Z=1)))
    Item=(Type=Bone,Name="Jacket_Fr_r_lookat")
    Primary=(Weight=1,Axis=(X=1,Y=0,Z=0),Target=(X=1,Y=0,Z=0),Kind=Location,Space=(Type=Bone,Name="dyn_Jacket_fr_TGT"))
    Secondary=(Weight=1,Axis=(X=0,Y=1,Z=0),Target=(X=1,Y=0,Z=0),Kind=Location,Space=(Type=Control,Name="pelvis_aim"))
    Weight=@n2.Value
   )
   AlphaInterp = &RigVMFunction_AlphaInterp(
    ClampMax=0
    ClampMin=1
    InRange=(Min=1,Max=0)
    InterpSpeedDecreasing=10
    InterpSpeedIncreasing=5
    OutRange=(Min=0,Max=0.1)
    Scale=1
    bClampResult=true
    Value=RigVMVariableNode() {variable=emote}.Value
   )
   > &RigUnit_ParentConstraint(
    AdvancedSettings=(RotationOrderForFilter=XZY)
    Child=(Type=Bone,Name="Jacket_Fr_l_lookat")
    Parents=((Item=(Type=Bone,Name="pelvis"),Weight=0.6),(Item=(Type=Bone,Name="spine_04"),Weight=0.4))
    Weight=@AlphaInterp.Result
   )
   > &RigUnit_ParentConstraint(
    AdvancedSettings=(RotationOrderForFilter=XZY)
    Child=(Type=Bone,Name="Jacket_Fr_r_lookat")
    Parents=((Item=(Type=Bone,Name="pelvis"),Weight=1))
    Weight=@AlphaInterp.Result
   )
   > &RigUnit_ModifyTransforms(
    ItemToModify="((Item=(Type=Bone,Name=\"Jacket_Fr_l_lookat\"),Transform=(Rotation=(X=0,Y=0,Z=0,W=1),Translation=(X=-5,Y=0,Z=3),Scale3D=(X=1,Y=1,Z=1))),(Item=(Type=Bone,Name=\"Jacket_Fr_r_lookat\"),Transform=(Rotation=(X=0,Y=0,Z=0,W=1),Translation=(X=-5,Y=0,Z=-3),Scale3D=(X=1 ...(truncated)"
    Weight=@AlphaInterp.Result
    ItemToModify.0.Transform.Rotation=&RigVMFunction_MathQuaternionFromEuler(Euler=(X=0,Y=-10,Z=-10)).Result
    ItemToModify.1.Transform.Rotation=&RigVMFunction_MathQuaternionFromEuler(Euler=(X=0,Y=10,Z=-10)).Result
   )
   &RigUnit_ParentConstraint(
    AdvancedSettings=(RotationOrderForFilter=XZY)
    Child=(Type=Bone,Name="Jacket_Fr_l_lookat")
    Parents=((Item=(Type=Bone,Name="pelvis"),Weight=0.6),(Item=(Type=Bone,Name="spine_04"),Weight=0.4))
    Weight=1
    ExecutePin=@n1.True
   )
   > &RigUnit_ParentConstraint(
    AdvancedSettings=(RotationOrderForFilter=XZY)
    Child=(Type=Bone,Name="Jacket_Fr_r_lookat")
    Parents=((Item=(Type=Bone,Name="pelvis"),Weight=1))
    Weight=1
   )
   n3 = &RigUnit_SphericalPoseReader(
    ActiveRegionScaleFactors=(PositiveWidth=3.5,NegativeWidth=0.5,PositiveHeight=3,NegativeHeight=3)
    ActiveRegionSize=0.1
    Debug=(bDrawDebug=false,DebugScale=25,DebugSegments=20,DebugThickness=0.25)
    DriverAxis=(X=-1,Y=0,Z=0)
    DriverItem=(Type=Bone,Name="thigh_l")
    FalloffRegionScaleFactors=(PositiveWidth=1,NegativeWidth=1,PositiveHeight=1,NegativeHeight=1)
    FalloffSize=0.45
    OptionalParentItem=(Type=Bone)
    RotationOffset=(X=90,Y=0,Z=0)
    ExecutePin=@SequenceAggregate_2.B
   )
   > &RigUnit_ModifyTransforms(
    ItemToModify="((Item=(Type=Bone,Name=\"Jacket_Fr_l_MSTR\"),Transform=(Rotation=(X=0,Y=0,Z=0,W=1),Translation=(X=-3,Y=-3,Z=5),Scale3D=(X=1,Y=1,Z=1))),(Item=(Type=Bone,Name=\"dyn_jacketSide_mstr_l\"),Transform=(Rotation=(X=0,Y=0,Z=0,W=1),Translation=(X=0,Y=-5,Z=0),Scale3D=(X= ...(truncated)"
    ItemToModify.0.Transform.Rotation=&RigVMFunction_MathQuaternionFromEuler(Euler=(X=15,Y=-15,Z=-15)).Result
    Weight=@n3.OutputParam
   )
   n4 = &RigUnit_SphericalPoseReader(
    ActiveRegionScaleFactors=(PositiveWidth=3.5,NegativeWidth=0.5,PositiveHeight=3,NegativeHeight=3)
    ActiveRegionSize=0.1
    Debug=(bDrawDebug=false,DebugScale=25,DebugSegments=20,DebugThickness=0.25)
    DriverAxis=(X=1,Y=0,Z=0)
    DriverItem=(Type=Bone,Name="thigh_r")
    FalloffRegionScaleFactors=(PositiveWidth=1,NegativeWidth=1,PositiveHeight=1,NegativeHeight=1)
    FalloffSize=0.45
    FlipHeightScaling=false
    FlipWidthScaling=true
    OptionalParentItem=(Type=Bone)
    RotationOffset=(X=-90,Y=0,Z=0)
    ExecutePin=@SequenceAggregate_2.C
   )
   > &RigUnit_ModifyTransforms(
    ItemToModify="((Item=(Type=Bone,Name=\"Jacket_Fr_r_MSTR\"),Transform=(Rotation=(X=0,Y=0,Z=0,W=1),Translation=(X=-3,Y=-3,Z=-5),Scale3D=(X=1,Y=1,Z=1))),(Item=(Type=Bone,Name=\"dyn_jacketSide_mstr_r\"),Transform=(Rotation=(X=0,Y=0,Z=0,W=1),Translation=(X=0,Y=0,Z=0),Scale3D=(X= ...(truncated)"
    ItemToModify.0.Transform.Rotation=&RigVMFunction_MathQuaternionFromEuler(Euler=(X=-15,Y=10,Z=-15)).Result
    Weight=@n4.OutputParam
   )
   > n5 = &RigUnit_SphericalPoseReader(
    ActiveRegionScaleFactors=(PositiveWidth=0.6,NegativeWidth=0.6,PositiveHeight=1,NegativeHeight=1)
    ActiveRegionSize=0.15
    Debug=(bDrawDebug=false,bDraw2D=false,bDrawLocalAxes=true,DebugScale=25,DebugSegments=20,DebugThickness=0.25)
    DriverAxis=(X=-1,Y=0,Z=0)
    DriverItem=(Type=Bone,Name="upperarm_r")
    FalloffRegionScaleFactors=(PositiveWidth=0.6,NegativeWidth=0.6,PositiveHeight=2,NegativeHeight=1)
    FalloffSize=0.45
    FlipHeightScaling=false
    FlipWidthScaling=true
    OptionalParentItem=(Type=Bone,Name="spine_04")
    RotationOffset=(X=-100,Y=25,Z=40)
   )
   > &RigUnit_ModifyTransforms(
    ItemToModify="((Item=(Type=Bone,Name=\"dyn_collar_fr\"),Transform=(Rotation=(X=0,Y=0,Z=0,W=1),Translation=(X=0,Y=0,Z=0),Scale3D=(X=1,Y=1,Z=1))),(Item=(Type=Bone,Name=\"dyn_jacket_collar_fr\"),Transform=(Rotation=(X=0,Y=0,Z=0,W=1),Translation=(X=0,Y=0,Z=0),Scale3D=(X=1,Y=1,Z ...(truncated)"
    ItemToModify.0.Transform.Rotation=&RigVMFunction_MathQuaternionFromEuler(Euler=(X=-25,Y=0,Z=-65)).Result
    ItemToModify.1.Transform.Rotation=&RigVMFunction_MathQuaternionFromEuler(Euler=(X=-50,Y=-15,Z=-70)).Result
    Weight=@n5.OutputParam
   )
   n6 = &RigVMFunction_ControlFlowBranch(ExecuteContext=@SequenceAggregate_2.D, Condition=RigVMVariableNode() {variable=IsCrouchingCtrl}.Value)
   &RigUnit_ModifyTransforms(
    ItemToModify=((Item=(Type=Bone,Name="dyn_beltbuckle_mstr"),Transform=(Rotation=(X=0,Y=0,Z=0,W=1),Translation=(X=0,Y=0,Z=2),Scale3D=(X=1,Y=1,Z=1))))
    Weight=1
    ExecutePin=@n6.False
    ItemToModify.0.Transform.Rotation=&RigVMFunction_MathQuaternionFromEuler(Euler=(X=0,Y=-5,Z=0), RotationOrder=XZY).Result
   )
   > n7 = &RigUnit_SphericalPoseReader(
    ActiveRegionScaleFactors=(PositiveWidth=1,NegativeWidth=1.7,PositiveHeight=1,NegativeHeight=1)
    ActiveRegionSize=0.1
    Debug=(bDrawDebug=false,bDraw2D=false,DebugScale=25,DebugSegments=20,DebugThickness=0.25)
    DriverAxis=(X=-1,Y=0,Z=0)
    DriverItem=(Type=Bone,Name="thigh_l")
    FalloffRegionScaleFactors=(PositiveWidth=1,NegativeWidth=5,PositiveHeight=2,NegativeHeight=2)
    FalloffSize=0.215
    OptionalParentItem=(Type=Bone,Name="pelvis")
    RotationOffset=(X=115,Y=0,Z=-20)
   )
   > &RigUnit_ModifyTransforms(
    ItemToModify=((Item=(Type=Bone,Name="dyn_beltbuckle_mstr"),Transform=(Rotation=(X=0,Y=0,Z=0,W=1),Translation=(X=0,Y=2,Z=3),Scale3D=(X=1,Y=1,Z=1))))
    ItemToModify.0.Transform.Rotation=&RigVMFunction_MathQuaternionFromEuler(Euler=(X=40,Y=-20,Z=0), RotationOrder=XZY).Result
    Weight=@n7.OutputParam
   )
   > &RigUnit_ModifyTransforms(
    ItemToModify=((Item=(Type=Bone,Name="dyn_buckelstrap"),Transform=(Rotation=(X=0,Y=0,Z=0,W=1),Translation=(X=0,Y=0,Z=0),Scale3D=(X=1,Y=1,Z=1))))
    ItemToModify.0.Transform.Rotation=&RigVMFunction_MathQuaternionFromEuler(Euler=(X=0,Y=0,Z=0), RotationOrder=YXZ).Result
    Weight=@n7.OutputParam
   )
   > n8 = &RigUnit_SphericalPoseReader(
    ActiveRegionScaleFactors=(PositiveWidth=1.5,NegativeWidth=1,PositiveHeight=1,NegativeHeight=1)
    ActiveRegionSize=0.1
    Debug=(bDrawDebug=false,bDraw2D=false,DebugScale=25,DebugSegments=20,DebugThickness=0.25)
    DriverAxis=(X=1,Y=0,Z=0)
    DriverItem=(Type=Bone,Name="thigh_r")
    FalloffRegionScaleFactors=(PositiveWidth=1,NegativeWidth=5,PositiveHeight=1.3,NegativeHeight=2.5)
    FalloffSize=0.2
    OptionalParentItem=(Type=Bone,Name="pelvis")
    RotationOffset=(X=-95,Y=180,Z=-10)
   )
   > &RigUnit_ModifyTransforms(
    ItemToModify=((Item=(Type=Bone,Name="dyn_beltbuckle_mstr"),Transform=(Rotation=(X=0,Y=0,Z=0,W=1),Translation=(X=0,Y=1,Z=2),Scale3D=(X=1,Y=1,Z=1))))
    ItemToModify.0.Transform.Rotation=&RigVMFunction_MathQuaternionFromEuler(Euler=(X=20,Y=10,Z=0), RotationOrder=XZY).Result
    Weight=@n8.OutputParam
   )
   > &RigUnit_ModifyTransforms(
    ItemToModify=((Item=(Type=Bone,Name="dyn_buckelstrap"),Transform=(Rotation=(X=0,Y=0,Z=0,W=1),Translation=(X=-1,Y=2,Z=5.5),Scale3D=(X=1,Y=1,Z=1))))
    ItemToModify.0.Transform.Rotation=&RigVMFunction_MathQuaternionFromEuler(Euler=(X=-70,Y=-30,Z=-40), RotationOrder=ZXY).Result
    Weight=@n8.OutputParam
   )
   > n9 = &RigUnit_SphericalPoseReader(
    ActiveRegionScaleFactors=(PositiveWidth=0.2,NegativeWidth=0.2,PositiveHeight=0.3,NegativeHeight=0.8)
    ActiveRegionSize=0.1
    Debug=(bDrawDebug=false,bDraw2D=false,DebugScale=25,DebugSegments=20,DebugThickness=0.25)
    DriverAxis=(X=1,Y=0,Z=0)
    DriverItem=(Type=Bone,Name="thigh_r")
    FalloffRegionScaleFactors=(PositiveWidth=1,NegativeWidth=1,PositiveHeight=0.3,NegativeHeight=0.3)
    FalloffSize=0.15
    OptionalParentItem=(Type=Bone,Name="pelvis")
    RotationOffset=(X=-95,Y=180,Z=-10)
   )
   > &RigUnit_ModifyTransforms(
    ItemToModify=((Item=(Type=Bone,Name="dyn_buckelstrap"),Transform=(Rotation=(X=0,Y=0,Z=0,W=1),Translation=(X=0,Y=0,Z=3),Scale3D=(X=1,Y=1,Z=1))))
    ItemToModify.0.Transform.Rotation=&RigVMFunction_MathQuaternionFromEuler(Euler=(X=0,Y=0,Z=-40), RotationOrder=ZXY).Result
    Weight=@n9.OutputParam
   )
   > n10 = &RigUnit_SphericalPoseReader(
    ActiveRegionScaleFactors=(PositiveWidth=0.3,NegativeWidth=0.3,PositiveHeight=0.5,NegativeHeight=0.5)
    ActiveRegionSize=0.1
    Debug=(bDrawDebug=false,bDraw2D=false,DebugScale=25,DebugSegments=20,DebugThickness=0.25)
    DriverAxis=(X=1,Y=0,Z=0)
    DriverItem=(Type=Bone,Name="thigh_r")
    FalloffRegionScaleFactors=(PositiveWidth=0.7,NegativeWidth=0.5,PositiveHeight=1,NegativeHeight=1)
    FalloffSize=0.15
    OptionalParentItem=(Type=Bone,Name="pelvis")
    RotationOffset=(X=-120,Y=66,Z=10)
   )
   > &RigUnit_ModifyTransforms(
    ItemToModify=((Item=(Type=Bone,Name="dyn_beltbuckle_mstr"),Transform=(Rotation=(X=0,Y=0,Z=0,W=1),Translation=(X=3,Y=0,Z=1),Scale3D=(X=1,Y=1,Z=1))))
    ItemToModify.0.Transform.Rotation=&RigVMFunction_MathQuaternionFromEuler(Euler=(X=0,Y=15,Z=-10), RotationOrder=XZY).Result
    Weight=@n10.OutputParam
   )
   > &RigUnit_ModifyTransforms(
    ItemToModify=((Item=(Type=Bone,Name="dyn_buckelstrap"),Transform=(Rotation=(X=0,Y=0,Z=0,W=1),Translation=(X=0,Y=0,Z=0),Scale3D=(X=1,Y=1,Z=1))))
    WeightMinimum=0
    ItemToModify.0.Transform.Rotation=&RigVMFunction_MathQuaternionFromEuler(Euler=(X=-20,Y=0,Z=-20), RotationOrder=XZY).Result
    Weight=@n10.OutputParam
   )
   > n11 = &RigUnit_SphericalPoseReader(
    ActiveRegionScaleFactors=(PositiveWidth=0.3,NegativeWidth=0.7,PositiveHeight=0.2,NegativeHeight=0.2)
    ActiveRegionSize=0.1
    Debug=(bDrawDebug=false,bDraw2D=false,DebugScale=25,DebugSegments=20,DebugThickness=0.25)
    DriverAxis=(X=0,Y=0,Z=-1)
    DriverItem=(Type=Bone,Name="thigh_r")
    FalloffRegionScaleFactors=(PositiveWidth=1,NegativeWidth=4,PositiveHeight=1,NegativeHeight=1)
    FalloffSize=0.05
    OptionalParentItem=(Type=Bone,Name="pelvis")
    RotationOffset=(X=0,Y=140,Z=20)
   )
   > &RigUnit_ModifyTransforms(
    ItemToModify=((Item=(Type=Bone,Name="dyn_beltbuckle_mstr"),Transform=(Rotation=(X=0,Y=0,Z=0,W=1),Translation=(X=1.5,Y=1.5,Z=1.5),Scale3D=(X=1,Y=1,Z=1))))
    WeightMinimum=0
    ItemToModify.0.Transform.Rotation=&RigVMFunction_MathQuaternionFromEuler(Euler=(X=0,Y=0,Z=0), RotationOrder=XZY).Result
    Weight=@n11.OutputParam
   )
   > &RigUnit_ModifyTransforms(
    ItemToModify=((Item=(Type=Bone,Name="dyn_buckelstrap"),Transform=(Rotation=(X=0,Y=0,Z=0,W=1),Translation=(X=1,Y=0,Z=-1),Scale3D=(X=1,Y=1,Z=1))))
    WeightMinimum=0
    ItemToModify.0.Transform.Rotation=&RigVMFunction_MathQuaternionFromEuler(Euler=(X=0,Y=20,Z=-20), RotationOrder=XZY).Result
    Weight=@n11.OutputParam
   )
   > n12 = &RigUnit_SphericalPoseReader(
    ActiveRegionScaleFactors=(PositiveWidth=0.3,NegativeWidth=1,PositiveHeight=0.2,NegativeHeight=0.2)
    ActiveRegionSize=0.1
    Debug=(bDrawDebug=false,bDraw2D=false,DebugScale=25,DebugSegments=20,DebugThickness=0.25)
    DriverAxis=(X=1,Y=0,Z=0)
    DriverItem=(Type=Bone,Name="thigh_r")
    FalloffRegionScaleFactors=(PositiveWidth=3.5,NegativeWidth=1,PositiveHeight=1,NegativeHeight=1)
    FalloffSize=0.05
    OptionalParentItem=(Type=Bone,Name="pelvis")
    RotationOffset=(X=-20,Y=-25,Z=10)
   )
   > &RigUnit_ModifyTransforms(
    ItemToModify=((Item=(Type=Bone,Name="dyn_buckelstrap"),Transform=(Rotation=(X=0,Y=0,Z=0,W=1),Translation=(X=0,Y=-1.5,Z=0),Scale3D=(X=1,Y=1,Z=1))))
    WeightMinimum=0
    ItemToModify.0.Transform.Rotation=&RigVMFunction_MathQuaternionFromEuler(Euler=(X=0,Y=0,Z=-20), RotationOrder=XZY).Result
    Weight=@n12.OutputParam
   )
   &RigUnit_ModifyTransforms(
    ItemToModify=((Item=(Type=Bone,Name="dyn_beltbuckle_mstr"),Transform=(Rotation=(X=0,Y=0,Z=0,W=1),Translation=(X=0,Y=1,Z=6),Scale3D=(X=1,Y=1,Z=1))))
    Weight=1
    ExecutePin=@n6.True
    ItemToModify.0.Transform.Rotation=&RigVMFunction_MathQuaternionFromEuler(Euler=(X=0,Y=-5,Z=0), RotationOrder=XZY).Result
   )
   > n13 = &RigUnit_SphericalPoseReader(
    ActiveRegionScaleFactors=(PositiveWidth=1,NegativeWidth=1.3,PositiveHeight=1,NegativeHeight=1)
    ActiveRegionSize=0.1
    Debug=(bDrawDebug=false,bDraw2D=false,bDrawLocalAxes=false,DebugScale=25,DebugSegments=20,DebugThickness=0.25)
    DriverAxis=(X=-1,Y=0,Z=0)
    DriverItem=(Type=Bone,Name="thigh_l")
    FalloffRegionScaleFactors=(PositiveWidth=1,NegativeWidth=5,PositiveHeight=2,NegativeHeight=2)
    FalloffSize=0.2
    OptionalParentItem=(Type=Bone,Name="pelvis")
    RotationOffset=(X=100,Y=0,Z=-40)
   )
   > &RigUnit_ModifyTransforms(
    ItemToModify=((Item=(Type=Bone,Name="dyn_beltbuckle_mstr"),Transform=(Rotation=(X=0,Y=0,Z=0,W=1),Translation=(X=1,Y=1,Z=0),Scale3D=(X=1,Y=1,Z=1))))
    ItemToModify.0.Transform.Rotation=&RigVMFunction_MathQuaternionFromEuler(Euler=(X=15,Y=-15,Z=0), RotationOrder=XZY).Result
    Weight=@n13.OutputParam
   )
   > &RigUnit_ModifyTransforms(
    ItemToModify=((Item=(Type=Bone,Name="dyn_buckelstrap"),Transform=(Rotation=(X=0,Y=0,Z=0,W=1),Translation=(X=0,Y=0,Z=0),Scale3D=(X=1,Y=1,Z=1))))
    Weight=1
   )
   > n14 = &RigUnit_SphericalPoseReader(
    ActiveRegionScaleFactors=(PositiveWidth=3.5,NegativeWidth=0.5,PositiveHeight=3,NegativeHeight=3)
    ActiveRegionSize=0.1
    Debug=(bDrawDebug=false,DebugScale=25,DebugSegments=20,DebugThickness=0.25)
    DriverAxis=(X=1,Y=0,Z=0)
    DriverItem=(Type=Bone,Name="thigh_r")
    FalloffRegionScaleFactors=(PositiveWidth=1,NegativeWidth=1,PositiveHeight=1,NegativeHeight=1)
    FalloffSize=0.45
    FlipHeightScaling=false
    FlipWidthScaling=true
    OptionalParentItem=(Type=Bone)
    RotationOffset=(X=-90,Y=0,Z=0)
   )
   > &RigUnit_ModifyTransforms(
    ItemToModify=((Item=(Type=Bone,Name="dyn_beltbuckle_mstr"),Transform=(Rotation=(X=0,Y=0,Z=0,W=1),Translation=(X=0,Y=0,Z=3),Scale3D=(X=1,Y=1,Z=1))))
    ItemToModify.0.Transform.Rotation=&RigVMFunction_MathQuaternionFromEuler(Euler=(X=0,Y=15,Z=0), RotationOrder=XZY).Result
    Weight=@n14.OutputParam
   )
   > &RigUnit_ModifyTransforms(
    ItemToModify=((Item=(Type=Bone,Name="dyn_buckelstrap_mstr"),Transform=(Rotation=(X=0,Y=0,Z=0,W=1),Translation=(X=-1,Y=1,Z=6),Scale3D=(X=1,Y=1,Z=1))))
    ItemToModify.0.Transform.Rotation=&RigVMFunction_MathQuaternionFromEuler(Euler=(X=90,Y=15,Z=-15), RotationOrder=YZX).Result
    Weight=@n14.OutputParam
   )
   n15 = &RigUnit_SphericalPoseReader(
    ActiveRegionScaleFactors=(PositiveWidth=0.5,NegativeWidth=1.3,PositiveHeight=1,NegativeHeight=0.5)
    ActiveRegionSize=0.1
    Debug=(bDrawDebug=True,bDraw2D=false,DebugScale=25,DebugSegments=20,DebugThickness=0.25)
    DriverAxis=(X=1,Y=0,Z=0)
    DriverItem=(Type=Bone,Name="thigh_r")
    ExecutePin=()
    FalloffRegionScaleFactors=(PositiveWidth=0.3,NegativeWidth=2.4,PositiveHeight=0.5,NegativeHeight=0.3)
    FalloffSize=0.25
    OptionalParentItem=(Type=Bone,Name="pelvis")
    RotationOffset=(X=-135,Y=170,Z=-20)
   )
   > &RigUnit_ModifyTransforms(
    ItemToModify=((Item=(Type=Bone,Name="dyn_beltbuckle_mstr"),Transform=(Rotation=(X=0,Y=0,Z=0,W=1),Translation=(X=0,Y=1,Z=3),Scale3D=(X=1,Y=1,Z=1))))
    ItemToModify.0.Transform.Rotation=&RigVMFunction_MathQuaternionFromEuler(Euler=(X=30,Y=15,Z=0), RotationOrder=XZY).Result
    Weight=@n15.OutputParam
   )
   > &RigUnit_ModifyTransforms(
    ItemToModify=((Item=(Type=Bone,Name="dyn_buckelstrap"),Transform=(Rotation=(X=0,Y=0,Z=0,W=1),Translation=(X=0,Y=0,Z=3),Scale3D=(X=1,Y=1,Z=1))))
    ItemToModify.0.Transform.Rotation=&RigVMFunction_MathQuaternionFromEuler(Euler=(X=-5,Y=-35,Z=-30), RotationOrder=XZY).Result
    Weight=@n15.OutputParam
   )
   > n16 = &RigUnit_SphericalPoseReader(
    ActiveRegionScaleFactors=(PositiveWidth=0.7,NegativeWidth=0.7,PositiveHeight=0.7,NegativeHeight=0.7)
    ActiveRegionSize=0.1
    Debug=(bDrawDebug=false,bDraw2D=false,DebugScale=25,DebugSegments=20,DebugThickness=0.25)
    DriverAxis=(X=0,Y=0,Z=-1)
    DriverItem=(Type=Bone,Name="thigh_r")
    FalloffRegionScaleFactors=(PositiveWidth=1.5,NegativeWidth=1.5,PositiveHeight=1,NegativeHeight=0.5)
    FalloffSize=0.25
    OptionalParentItem=(Type=Bone,Name="pelvis")
    RotationOffset=(X=180,Y=-55,Z=-20)
   )
   > &RigUnit_ModifyTransforms(
    ItemToModify=((Item=(Type=Bone,Name="dyn_beltbuckle_mstr"),Transform=(Rotation=(X=0,Y=0,Z=0,W=1),Translation=(X=2,Y=0,Z=0),Scale3D=(X=1,Y=1,Z=1))))
    Weight=@n16.OutputParam
   )
   > &RigUnit_ModifyTransforms(
    ItemToModify=((Item=(Type=Bone,Name="dyn_buckelstrap"),Transform=(Rotation=(X=0,Y=0,Z=0,W=1),Translation=(X=-4,Y=0,Z=2.5),Scale3D=(X=1,Y=1,Z=1))))
    ItemToModify.0.Transform.Rotation=&RigVMFunction_MathQuaternionFromEuler(Euler=(X=-40,Y=0,Z=-30), RotationOrder=YZX).Result
    Weight=@n16.OutputParam
   )
   &RigVMFunction_MathQuaternionFromEuler(Euler=(X=0,Y=-25,Z=0))
   &RigVMFunction_MathQuaternionFromEuler(Euler=(X=0,Y=0,Z=0))
   &RigVMFunction_MathQuaternionFromEuler(Euler=(X=0,Y=0,Z=0))
   &RigVMFunction_MathQuaternionFromEuler(Euler=(X=0,Y=-25,Z=0))
   &RigVMFunction_MathQuaternionFromEuler(Euler=(X=-30,Y=-40,Z=-30), RotationOrder=XZY)
   &RigVMFunction_MathQuaternionFromEuler(Euler=(X=0,Y=120,Z=-90))
   &RigVMFunction_MathQuaternionFromEuler(Euler=(X=40,Y=20,Z=-20), RotationOrder=XZY)
   &RigVMFunction_MathQuaternionFromEuler(Euler=(X=-20,Y=-40,Z=-50), RotationOrder=XZY)
   &RigUnit_SphericalPoseReader(
    ActiveRegionScaleFactors=(PositiveWidth=1,NegativeWidth=1.85,PositiveHeight=1,NegativeHeight=0.5)
    ActiveRegionSize=0.1
    Debug=(bDrawDebug=True,bDraw2D=false,DebugScale=25,DebugSegments=20,DebugThickness=0.25)
    DriverAxis=(X=1,Y=0,Z=0)
    DriverItem=(Type=Bone,Name="thigh_r")
    ExecutePin=()
    FalloffRegionScaleFactors=(PositiveWidth=1,NegativeWidth=3.5,PositiveHeight=0.4,NegativeHeight=0.2)
    FalloffSize=0.25
    OptionalParentItem=(Type=Bone,Name="pelvis")
    RotationOffset=(X=-100,Y=180,Z=-20)
   )
  ]}
]
hierarchy [
 root {loc=<0 0 0>, rot=<0 0 0 1>}
 pelvis {parent=root, loc=<0 -0.7071 93.1191>, rot=<0.0127 0.707 -0.0127 -0.707>}
 spine_01 {parent=pelvis, loc=<6.7494 -0.6388 0>, rot=<0 0 0.034 -0.9994>}
 spine_02 {parent=spine_01, loc=<7.2518 -0.035 0>, rot=<0 0 0.0238 -0.9997>}
 spine_03 {parent=spine_02, loc=<6.0104 0 0>, rot=<0 0 0.0149 -0.9999>}
 spine_04 {parent=spine_03, loc=<8.1024 0 0>, rot=<0 0 -0.0132 -0.9999>}
 spine_05 {parent=spine_04, loc=<10.0146 0 0>, rot=<0 0 -0.0015 -1>}
 neck_01 {parent=spine_05, loc=<11.3138 0.1529 0>, rot=<0 0 -0.2003 -0.9797>}
 neck_02 {parent=neck_01, loc=<4.5627 0 0>, rot=<0 0 0.0504 -0.9987>}
 head {parent=neck_02, loc=<5.3914 0 0>, rot=<0 0 0.115 -0.9934>}
 clavicle_l {parent=spine_05, loc=<4.3575 -0.1466 -2.3223>, rot=<0.0465 -0.7289 -0.009 -0.6829>}
 upperarm_l {parent=clavicle_l, loc=<15.5111 0 0>, rot=<0.0226 -0.3234 -0.0088 -0.9459>}
 lowerarm_l {parent=upperarm_l, loc=<26.6514 0 0>, rot=<0.0391 -0.0099 -0.2453 -0.9686>}
 hand_l {parent=lowerarm_l, loc=<22.3491 0 0>, rot=<-0.7081 -0.0507 0.0975 -0.6975>}
 thumb_01_l {parent=hand_l, loc=<1.3118 -0.3191 -2.459>, rot=<0.6546 -0.4042 0.0664 -0.6354>}
 thumb_02_l {parent=thumb_01_l, loc=<5.9336 -0.0448 0.0401>, rot=<-0.0037 0.0213 0.298 -0.9543>}
 thumb_03_l {parent=thumb_02_l, loc=<2.2223 0.3682 -0.0907>, rot=<0.0118 -0.013 0.0066 -0.9998>}
 index_metacarpal_l {parent=hand_l, loc=<1.6535 0.0183 -2.96>, rot=<0.0158 -0.0124 -0.0173 -0.9996>}
 index_01_l {parent=index_metacarpal_l, loc=<8.3008 0.0357 -0.0524>, rot=<-0.0044 0.0485 0.0924 -0.9945>}
 index_02_l {parent=index_01_l, loc=<3.3587 -0.3004 0.0452>, rot=<-0.002 0.0027 0.3225 -0.9466>}
 index_03_l {parent=index_02_l, loc=<2.5578 -0.123 -0.0252>, rot=<-0.002 0.0067 0.2558 -0.9667>}
 middle_metacarpal_l {parent=hand_l, loc=<2.1856 -0.1725 -0.5984>, rot=<0.006 0.0168 -0.0329 -0.9993>}
 middle_01_l {parent=middle_metacarpal_l, loc=<7.3275 -0.2447 0.0129>, rot=<-0.0208 0.0042 0.0858 -0.9961>}
 middle_02_l {parent=middle_01_l, loc=<4.0123 0.1401 0.1175>, rot=<-0.0218 0.0258 0.3512 -0.9357>}
 middle_03_l {parent=middle_02_l, loc=<2.6077 0.0534 0.1373>, rot=<0.0082 0.0188 0.1708 -0.9851>}
 ring_metacarpal_l {parent=hand_l, loc=<2.2074 -0.3625 1.3638>, rot=<-0.0284 0.0722 -0.0458 -0.9959>}
 ring_01_l {parent=ring_metacarpal_l, loc=<7.0295 -0.6197 -0.1446>, rot=<-0.0489 -0.0252 0.1892 -0.9804>}
 ring_02_l {parent=ring_01_l, loc=<3.6174 0.4647 -0.1027>, rot=<0 0 0.2806 -0.9598>}
 ring_03_l {parent=ring_02_l, loc=<2.4152 0.114 0.0199>, rot=<0 0 0.1189 -0.9929>}
 pinky_metacarpal_l {parent=hand_l, loc=<1.7616 -0.9753 3.3814>, rot=<-0.0397 0.0814 -0.069 -0.9935>}
 pinky_01_l {parent=pinky_metacarpal_l, loc=<7.1709 -0.4831 -0.1094>, rot=<-0.0654 -0.0035 0.2242 -0.9723>}
 pinky_02_l {parent=pinky_01_l, loc=<2.8274 0.0893 -0.1043>, rot=<0.0037 0.0011 0.2854 -0.9584>}
 pinky_03_l {parent=pinky_02_l, loc=<2.2046 0.0307 -0.0078>, rot=<-0.0002 0.0098 0.1445 -0.9895>}
 deform_wrist_dn_base_l {parent=hand_l, loc=<0 -2.2342 0>, rot=<0 0 0 -1>}
 deform_wrist_dn_l {parent=deform_wrist_dn_base_l, loc=<0 0 0>, rot=<0 0 0 1>}
 deform_wrist_up_base_l {parent=hand_l, loc=<0 2.2682 0>, rot=<0 0 0 -1>}
 deform_wrist_up_l {parent=deform_wrist_up_base_l, loc=<0 0 0>, rot=<0 0 0 1>}
 lowerarm_twist_02_l {parent=lowerarm_l, loc=<9.6199 0 0>, rot=<0 0 0 -1>}
 lowerarm_twist_01_l {parent=lowerarm_l, loc=<17.8549 0 0>, rot=<0 0 0 -1>}
 deform_elbow_in_base_l {parent=lowerarm_l, loc=<0.2112 5.259 0.8274>, rot=<0 0 0 -1>}
 deform_elbow_in_l {parent=deform_elbow_in_base_l, loc=<0 0 0>, rot=<0 0 0 1>}
 deform_elbow_out_base_l {parent=lowerarm_l, loc=<0 -4 0>, rot=<0 0 0 -1>}
 deform_elbow_out_l {parent=deform_elbow_out_base_l, loc=<0 0 0>, rot=<0 0 0 1>}
 upperarm_twist_01_l {parent=upperarm_l, loc=<0.0041 0 0>, rot=<0 0 0 -1>}
 deform_upperarm_bk_base_l {parent=upperarm_twist_01_l, loc=<1.0813 -7.2193 -1.9453>, rot=<0 0 0 1>}
 deform_upperarm_bk_l {parent=deform_upperarm_bk_base_l, loc=<0 0 0>, rot=<0 0 0 1>}
 deform_upperarm_fr_base_l {parent=upperarm_twist_01_l, loc=<-0.3691 7.0506 -2.765>, rot=<0 0 0 1>}
 deform_upperarm_fr_l {parent=deform_upperarm_fr_base_l, loc=<0 0 0>, rot=<0 0 0 1>}
 deform_upperarm_up_base_l {parent=upperarm_twist_01_l, loc=<-3.2216 -1.1391 3.3311>, rot=<0 0 0 1>}
 deform_upperarm_up_l {parent=deform_upperarm_up_base_l, loc=<0 0 0>, rot=<0 0 0 1>}
 upperarm_twist_02_l {parent=upperarm_l, loc=<13.7941 0 0>, rot=<0 0 0 -1>}
 clavicle_r {parent=spine_05, loc=<4.3575 -0.1466 2.3223>, rot=<0.7289 0.0465 -0.6829 0.009>}
 upperarm_r {parent=clavicle_r, loc=<-15.5111 0 0>, rot=<0.0226 -0.3234 -0.0088 -0.9459>}
 lowerarm_r {parent=upperarm_r, loc=<-26.6514 0 0>, rot=<0.0391 -0.0099 -0.2453 -0.9686>}
 hand_r {parent=lowerarm_r, loc=<-22.3491 0 0>, rot=<-0.7081 -0.0507 0.0975 -0.6975>}
 thumb_01_r {parent=hand_r, loc=<-1.3118 0.3191 2.459>, rot=<0.6546 -0.4042 0.0664 -0.6354>}
 thumb_02_r {parent=thumb_01_r, loc=<-5.9336 0.0448 -0.0401>, rot=<-0.0037 0.0213 0.298 -0.9543>}
 thumb_03_r {parent=thumb_02_r, loc=<-2.2223 -0.3682 0.0907>, rot=<0.0118 -0.013 0.0066 -0.9998>}
 index_metacarpal_r {parent=hand_r, loc=<-1.6535 -0.0183 2.96>, rot=<0.0158 -0.0124 -0.0173 -0.9996>}
 index_01_r {parent=index_metacarpal_r, loc=<-8.3008 -0.0357 0.0524>, rot=<-0.0044 0.0485 0.0924 -0.9945>}
 index_02_r {parent=index_01_r, loc=<-3.3587 0.3004 -0.0452>, rot=<-0.002 0.0027 0.3225 -0.9466>}
 index_03_r {parent=index_02_r, loc=<-2.5578 0.123 0.0252>, rot=<-0.002 0.0067 0.2558 -0.9667>}
 middle_metacarpal_r {parent=hand_r, loc=<-2.1856 0.1725 0.5984>, rot=<0.006 0.0168 -0.0329 -0.9993>}
 middle_01_r {parent=middle_metacarpal_r, loc=<-7.3275 0.2447 -0.0129>, rot=<-0.0208 0.0042 0.0858 -0.9961>}
 middle_02_r {parent=middle_01_r, loc=<-4.0123 -0.1401 -0.1175>, rot=<-0.0218 0.0258 0.3512 -0.9357>}
 middle_03_r {parent=middle_02_r, loc=<-2.6077 -0.0534 -0.1373>, rot=<0.0082 0.0188 0.1708 -0.9851>}
 ring_metacarpal_r {parent=hand_r, loc=<-2.2074 0.3625 -1.3638>, rot=<-0.0284 0.0722 -0.0458 -0.9959>}
 ring_01_r {parent=ring_metacarpal_r, loc=<-7.0295 0.6197 0.1446>, rot=<-0.0489 -0.0252 0.1892 -0.9804>}
 ring_02_r {parent=ring_01_r, loc=<-3.6174 -0.4647 0.1027>, rot=<0 0 0.2806 -0.9598>}
 ring_03_r {parent=ring_02_r, loc=<-2.4152 -0.114 -0.0199>, rot=<0 0 0.1189 -0.9929>}
 pinky_metacarpal_r {parent=hand_r, loc=<-1.7616 0.9753 -3.3814>, rot=<-0.0397 0.0814 -0.069 -0.9935>}
 pinky_01_r {parent=pinky_metacarpal_r, loc=<-7.1709 0.4831 0.1094>, rot=<-0.0654 -0.0035 0.2242 -0.9723>}
 pinky_02_r {parent=pinky_01_r, loc=<-2.8274 -0.0893 0.1043>, rot=<0.0037 0.0011 0.2854 -0.9584>}
 pinky_03_r {parent=pinky_02_r, loc=<-2.2046 -0.0307 0.0078>, rot=<-0.0002 0.0098 0.1445 -0.9895>}
 deform_wrist_dn_base_r {parent=hand_r, loc=<-0.0004 2.2345 -0.0001>, rot=<0 0 0 -1>}
 deform_wrist_dn_r {parent=deform_wrist_dn_base_r, loc=<0 0 0>, rot=<0 0 0 1>}
 deform_wrist_up_base_r {parent=hand_r, loc=<0.0003 -2.2684 0.0001>, rot=<0 0 0 -1>}
 deform_wrist_up_r {parent=deform_wrist_up_base_r, loc=<0 0 0>, rot=<0 0 0 1>}
 lowerarm_twist_02_r {parent=lowerarm_r, loc=<-9.6199 0 0>, rot=<0 0 0 -1>}
 lowerarm_twist_01_r {parent=lowerarm_r, loc=<-17.8549 0 0>, rot=<0 0 0 -1>}
 deform_elbow_in_base_r {parent=lowerarm_r, loc=<-0.2111 -5.2591 -0.8275>, rot=<0 0 0 -1>}
 deform_elbow_in_r {parent=deform_elbow_in_base_r, loc=<-0.0001 0 0.0002>, rot=<0 0 0 1>}
 deform_elbow_out_base_r {parent=lowerarm_r, loc=<0 4 -0.0001>, rot=<0 0 0 -1>}
 deform_elbow_out_r {parent=deform_elbow_out_base_r, loc=<0 0 0>, rot=<0 0 0 1>}
 upperarm_twist_01_r {parent=upperarm_r, loc=<-0.0041 0 0>, rot=<0 0 0 -1>}
 deform_upperarm_bk_base_r {parent=upperarm_twist_01_r, loc=<-1.0814 7.2192 1.9455>, rot=<0 0 0 1>}
 deform_upperarm_bk_r {parent=deform_upperarm_bk_base_r, loc=<0 0 0>, rot=<0 0 0 1>}
 deform_upperarm_fr_base_r {parent=upperarm_twist_01_r, loc=<0.3689 -7.0506 2.7652>, rot=<0 0 0 1>}
 deform_upperarm_fr_r {parent=deform_upperarm_fr_base_r, loc=<0 0 0>, rot=<0 0 0 1>}
 deform_upperarm_up_base_r {parent=upperarm_twist_01_r, loc=<3.2214 1.1391 -3.3308>, rot=<0 0 0 1>}
 deform_upperarm_up_r {parent=deform_upperarm_up_base_r, loc=<0 0 0>, rot=<0 0 0 -1>}
 upperarm_twist_02_r {parent=upperarm_r, loc=<-13.7941 0 0>, rot=<0 0 0 -1>}
 attach_backpack {parent=spine_05, loc=<-0.123 -13.473 0>, rot=<-0.0093 -0.707 0.0093 -0.707>}
 attach_cape {parent=spine_05, loc=<9.017 -11.714 0>, rot=<0.0616 -0.7044 -0.0616 -0.7044>}
 dyn_collar_fr {parent=spine_05, loc=<9.5082 7.1029 4.9472>, rot=<-0.6887 0.1601 -0.3902 -0.5897>}
 dyn_jacket_collar_fr {parent=spine_05, loc=<8.2329 6.3232 8.0657>, rot=<0.8408 0.0764 0.379 0.3789>}
 dyn_collar_bk_MSTR {parent=spine_05, loc=<14.4636 -4.9708 -0.0736>, rot=<0.5196 -0.4795 0.4795 -0.5196>}
 dyn_collar_bk {parent=dyn_collar_bk_MSTR, loc=<0 0 0>, rot=<0 0 0 1>}
 dyn_collar_l {parent=spine_05, loc=<14.0576 -2.6647 -5.7587>, rot=<0.3285 -0.6647 0.1774 -0.6472>}
 deform_pec_aim_base_l {parent=spine_04, loc=<7.8732 12.6388 -2.4123>, rot=<0.0102 -0.4602 0.3979 -0.7936>}
 deform_pec_aim_l {parent=deform_pec_aim_base_l, loc=<0 0 0>, rot=<0 0 0 1>}
 deform_pec_base_l {parent=deform_pec_aim_l, loc=<10.766 0 0>, rot=<0 0 0 1>}
 deform_pec_l {parent=deform_pec_base_l, loc=<0.055 0 0>, rot=<0 0 0 1>}
 deform_pec_aim_base_r {parent=spine_04, loc=<7.8727 12.6388 2.4123>, rot=<-0.4602 -0.0102 0.7936 0.3979>}
 deform_pec_aim_r {parent=deform_pec_aim_base_r, loc=<0 0 0>, rot=<0 0 0 1>}
 deform_pec_base_r {parent=deform_pec_aim_r, loc=<-10.7662 -0.0006 -0.0007>, rot=<0 0 0 1>}
 deform_pec_r {parent=deform_pec_base_r, loc=<0 0 0>, rot=<0 0 0 1>}
 Jacket_Fr_r_lookat {parent=spine_04, loc=<-4.7213 12.8284 8.5108>, rot=<0.0856 -0.9957 0.0001 -0.0361>}
 Jacket_Fr_r_MSTR {parent=Jacket_Fr_r_lookat, loc=<0 0 0>, rot=<0.0017 -0.0458 -0.0505 -0.9977>}
 Jacket_Fr_r_01 {parent=Jacket_Fr_r_MSTR, loc=<0 0 0>, rot=<0 0 0 1>}
 Jacket_Fr_r_02 {parent=Jacket_Fr_r_01, loc=<9.3452 0 0>, rot=<-0.0003 -0.021 -0.0025 -0.9998>}
 Jacket_Fr_r_03 {parent=Jacket_Fr_r_02, loc=<11.1594 -0.0066 0.0554>, rot=<0 0.0066 0.0241 -0.9997>}
 Jacket_Fr_r_04 {parent=Jacket_Fr_r_03, loc=<13.1551 0.1261 0.0188>, rot=<-0.0001 0.0024 0.0167 -0.9999>}
 Jacket_Fr_r_05 {parent=Jacket_Fr_r_04, loc=<13.9742 0.2695 -0.0022>, rot=<-0.0007 0.0129 -0.0128 -0.9998>}
 Jacket_Fr_r_06 {parent=Jacket_Fr_r_05, loc=<15.6885 0.1204 -0.1521>, rot=<-0.0001 0.0025 -0.0071 -1>}
 Jacket_Fr_r_07 {parent=Jacket_Fr_r_06, loc=<17.6889 0.016 -0.1891>, rot=<0 -0.004 -0.0135 -0.9999>}
 Jacket_Fr_l_lookat {parent=spine_04, loc=<-4.7213 12.8284 -8.5108>, rot=<0.0856 -0.9957 -0.0001 0.0361>}
 Jacket_Fr_l_MSTR {parent=Jacket_Fr_l_lookat, loc=<0 0 0>, rot=<-0.0017 0.0458 -0.0505 -0.9977>}
 Jacket_Fr_l_01 {parent=Jacket_Fr_l_MSTR, loc=<0 0 0>, rot=<0 0 0 1>}
 Jacket_Fr_l_02 {parent=Jacket_Fr_l_01, loc=<9.3452 0 0>, rot=<0.0003 0.021 -0.0025 -0.9998>}
 Jacket_Fr_l_03 {parent=Jacket_Fr_l_02, loc=<11.1594 -0.0066 -0.0554>, rot=<0 -0.0066 0.0241 -0.9997>}
 Jacket_Fr_l_04 {parent=Jacket_Fr_l_03, loc=<13.1551 0.1261 -0.0188>, rot=<0.0001 -0.0024 0.0167 -0.9999>}
 Jacket_Fr_l_05 {parent=Jacket_Fr_l_04, loc=<13.9742 0.2695 0.0022>, rot=<0.0007 -0.0129 -0.0128 -0.9998>}
 Jacket_Fr_l_06 {parent=Jacket_Fr_l_05, loc=<15.6885 0.1204 0.1521>, rot=<0.0001 -0.0025 -0.0071 -1>}
 Jacket_Fr_l_07 {parent=Jacket_Fr_l_06, loc=<17.6889 0.016 0.1891>, rot=<0 0.004 -0.0135 -0.9999>}
 dyn_jacketBack_MSTR {parent=spine_02, loc=<-4.0183 -14.3672 0.1238>, rot=<0.0544 -0.9985 0.0001 0.0053>}
 dyn_jacketBack_01 {parent=dyn_jacketBack_MSTR, loc=<0 0 0>, rot=<0 0 0 1>}
 dyn_jacketBack_02 {parent=dyn_jacketBack_01, loc=<17.8654 0 0>, rot=<0.0001 -0.0056 -0.0051 -1>}
 dyn_jacketBack_03_a {parent=dyn_jacketBack_02, loc=<18.8636 0.2075 3.4179>, rot=<-0.0006 0.0476 -0.0061 -0.9988>}
 dyn_jacketBack_04_a {parent=dyn_jacketBack_03_a, loc=<14.9451 0 0>, rot=<0.0001 -0.0079 0.0017 -1>}
 dyn_jacketBack_05_a {parent=dyn_jacketBack_04_a, loc=<16.1618 0 0>, rot=<0 0.0034 0.0032 -1>}
 dyn_jacketBack_03_b {parent=dyn_jacketBack_02, loc=<18.8513 0.2173 -3.0914>, rot=<0.0006 -0.0292 0.0016 -0.9996>}
 dyn_jacketBack_04_b {parent=dyn_jacketBack_03_b, loc=<15.1278 0 0>, rot=<0 -0.0006 0.0003 -1>}
 dyn_jacketBack_05_b {parent=dyn_jacketBack_04_b, loc=<16.1054 0 0>, rot=<0 -0.0005 -0.0059 -1>}
 dyn_jacketBack_mstr_r {parent=spine_02, loc=<-4.699 -9.9496 14.107>, rot=<0.0568 -0.487 0.8714 -0.0171>}
 dyn_jacketBack_01_r {parent=dyn_jacketBack_mstr_r, loc=<0 0 0>, rot=<0 0 -0.022 -0.9998>}
 dyn_jacketBack_02_r {parent=dyn_jacketBack_01_r, loc=<17 0 0>, rot=<-0.0271 0.0057 -0.0002 -0.9996>}
 dyn_jacketBack_03_r {parent=dyn_jacketBack_02_r, loc=<17 0 0>, rot=<-0.001 -0.0005 0 -1>}
 dyn_jacketBack_04_r {parent=dyn_jacketBack_03_r, loc=<17 0 0>, rot=<0.0016 -0.0002 0 -1>}
 dyn_jacketBack_05_r {parent=dyn_jacketBack_04_r, loc=<17 0 0>, rot=<0 0 0 1>}
 dyn_jacketBack_mstr_l {parent=spine_02, loc=<-4.6987 -9.9495 -14.107>, rot=<0.009 -0.8724 0.4862 -0.0498>}
 dyn_jacketBack_01_l {parent=dyn_jacketBack_mstr_l, loc=<0 0 0>, rot=<0 0 0.0219 -0.9998>}
 dyn_jacketBack_02_l {parent=dyn_jacketBack_01_l, loc=<17 0 0>, rot=<0.0271 0.0057 0.0002 -0.9996>}
 dyn_jacketBack_03_l {parent=dyn_jacketBack_02_l, loc=<17 0 0>, rot=<-0.0184 0.0097 -0.0002 -0.9998>}
 dyn_jacketBack_04_l {parent=dyn_jacketBack_03_l, loc=<17 0 0>, rot=<-0.0016 -0.0003 0 -1>}
 dyn_jacketBack_05_l {parent=dyn_jacketBack_04_l, loc=<17 0 0>, rot=<0 0 0 1>}
 dyn_jacketSide_mstr_l {parent=spine_01, loc=<5.1564 0.7398 -17.5987>, rot=<0.0329 -0.9979 -0.0475 -0.0288>}
 dyn_jacketSide_01_l {parent=dyn_jacketSide_mstr_l, loc=<0 0 0>, rot=<-0.0475 0 0 -0.9989>}
 dyn_jacketSide_02_l {parent=dyn_jacketSide_01_l, loc=<17 0 0>, rot=<-0.0001 0.0017 0.0157 -0.9999>}
 dyn_jacketSide_03_l {parent=dyn_jacketSide_02_l, loc=<17 0 0>, rot=<0.0009 -0.0141 0.0037 -0.9999>}
 dyn_jacketSide_04_l {parent=dyn_jacketSide_03_l, loc=<17.0048 0 0>, rot=<-0.0008 0.0121 -0.0055 -0.9999>}
 dyn_jacketSide_05_l {parent=dyn_jacketSide_04_l, loc=<16.4842 0 0>, rot=<-0.0001 0.002 0.0017 -1>}
 dyn_jacketSide_mstr_r {parent=spine_01, loc=<5.1579 0.7381 17.6144>, rot=<0.0349 0.0317 0.9983 -0.0331>}
 dyn_jacketSide_01_r {parent=dyn_jacketSide_mstr_r, loc=<0 0 0>, rot=<0 0 0 1>}
 dyn_jacketSide_02_r {parent=dyn_jacketSide_01_r, loc=<17 0 0>, rot=<0.9994 0.0124 0.0017 0.0318>}
 dyn_jacketSide_03_r {parent=dyn_jacketSide_02_r, loc=<17 0 0>, rot=<-0.0007 0.0122 0.0012 -0.9999>}
 dyn_jacketSide_04_r {parent=dyn_jacketSide_03_r, loc=<17.0028 0 0>, rot=<0.0004 -0.008 -0.0031 -1>}
 dyn_jacketSide_05_r {parent=dyn_jacketSide_04_r, loc=<16.4845 0 0>, rot=<0.0002 -0.0041 0.002 -1>}
 thigh_l {parent=pelvis, loc=<-2.5693 -0.0109 -7.7501>, rot=<0.0012 0.0572 0.0155 -0.9982>}
 calf_l {parent=thigh_l, loc=<-39.0484 0 0>, rot=<0.0006 -0.0161 -0.0391 -0.9991>}
 foot_l {parent=calf_l, loc=<-42.561 0.1077 0.4292>, rot=<0.0418 -0.0455 0.0343 -0.9975>}
 ball_l {parent=foot_l, loc=<-6.1599 13.8264 0.6806>, rot=<0.0001 -0.0001 -0.7063 -0.7079>}
 calf_twist_02_l {parent=calf_l, loc=<-18.8064 -0.0967 0.0397>, rot=<0.0237 0.0095 -0.0035 -0.9997>}
 calf_twist_01_l {parent=calf_l, loc=<-35.813 -0.0967 0.0397>, rot=<0.0237 0.0095 -0.0035 -0.9997>}
 deform_knee_in_base_l {parent=calf_l, loc=<1.298 -5.2957 -0.0496>, rot=<0 0 0 -1>}
 deform_knee_in_l {parent=deform_knee_in_base_l, loc=<0 0 0>, rot=<0 0 0 1>}
 deform_knee_out_base_l {parent=calf_l, loc=<0 4 0>, rot=<0 0 0 -1>}
 deform_knee_out_l {parent=deform_knee_out_base_l, loc=<0 0 0>, rot=<0 0 0 1>}
 thigh_twist_01_l {parent=thigh_l, loc=<-21.6303 0 0>, rot=<0.0013 0 0 -1>}
 deform_groin_base_l {parent=thigh_l, loc=<-2.1908 8.1142 -1.7442>, rot=<0 0 0 -1>}
 deform_groin_l {parent=deform_groin_base_l, loc=<0 0 0>, rot=<0 0 0 1>}
 thigh_r {parent=pelvis, loc=<-2.5693 -0.0109 7.7501>, rot=<0.0572 -0.0012 0.9982 0.0155>}
 calf_r {parent=thigh_r, loc=<39.0484 0 0>, rot=<0.0006 -0.0161 -0.0391 -0.9991>}
 foot_r {parent=calf_r, loc=<42.561 -0.1077 -0.4292>, rot=<0.0418 -0.0455 0.0343 -0.9975>}
 ball_r {parent=foot_r, loc=<6.1599 -13.8264 -0.6806>, rot=<0.0001 -0.0001 -0.7063 -0.7079>}
 calf_twist_02_r {parent=calf_r, loc=<18.8064 0.0967 -0.0397>, rot=<0.0237 0.0095 -0.0035 -0.9997>}
 calf_twist_01_r {parent=calf_r, loc=<35.813 0.0967 -0.0397>, rot=<0.0237 0.0095 -0.0035 -0.9997>}
 deform_knee_in_base_r {parent=calf_r, loc=<-1.298 5.2957 0.0496>, rot=<0 0 0 -1>}
 deform_knee_in_r {parent=deform_knee_in_base_r, loc=<0 0 0>, rot=<0 0 0 1>}
 deform_knee_out_base_r {parent=calf_r, loc=<0 -4 0>, rot=<0 0 0 -1>}
 deform_knee_out_r {parent=deform_knee_out_base_r, loc=<0 0 0>, rot=<0 0 0 1>}
 thigh_twist_01_r {parent=thigh_r, loc=<21.6303 0 0>, rot=<0.0013 0 0 -1>}
 deform_groin_base_r {parent=thigh_r, loc=<2.1908 -8.1142 1.7442>, rot=<0 0 0 -1>}
 deform_groin_r {parent=deform_groin_base_r, loc=<0 0 0>, rot=<0 0 0 1>}
 deform_glute_base_l {parent=pelvis, loc=<-8.8064 -8.8222 -8.4287>, rot=<0.0275 0.0501 -0.4472 -0.8926>}
 deform_glute_l {parent=deform_glute_base_l, loc=<0 0 0>, rot=<0 0 0 1>}
 deform_glute_base_r {parent=pelvis, loc=<-8.8062 -8.8222 8.4287>, rot=<0.0501 -0.0275 0.8926 -0.4472>}
 deform_glute_r {parent=deform_glute_base_r, loc=<0 0 0>, rot=<0 0 0 1>}
 pelvis_aim {controlType=EULER_TRANSFORM, parent=pelvis, type=Control, loc=<0 40 0>, rot=<0 0 0 1>}
 dyn_beltbuckle_mstr {parent=pelvis, loc=<-4.882 11.5097 0>, rot=<-0.0127 -0.707 0.0127 -0.707>}
 dyn_buckelstrap_mstr {parent=dyn_beltbuckle_mstr, loc=<-4.7033 -0.1015 -4.1158>, rot=<0 0 0 1>}
 dyn_buckelstrap {parent=dyn_buckelstrap_mstr, loc=<0 0 0>, rot=<-0.2167 0.0189 0.9724 -0.085>}
 dyn_tube {parent=pelvis, loc=<-2.5002 10.7329 8.5661>, rot=<0.6821 0.431 0.4845 0.3381>}
 dyn_box02 {parent=pelvis, loc=<-1.3987 8.8025 12.1569>, rot=<0.7339 0.3377 0.5339 0.2496>}
 dyn_box03 {parent=pelvis, loc=<-0.0417 4.2265 15.1716>, rot=<0.7714 0.2152 0.5918 0.0917>}
 dyn_box01 {parent=pelvis, loc=<-2.5193 10.8793 -8.7356>, rot=<-0.3597 -0.48 -0.4507 -0.6611>}
 dyn_Jacket_fr_TGT {parent=root, loc=<0 0 0>, rot=<0 0 0 1>}
 FloorCOL {parent=root, loc=<0 0 0>, rot=<0 0 0 1>}
]
variables [
 IsCrouchingCtrl {type=bool, default=False}
 Pickaxe {type=bool, default=False}
 emote {type=double, default=0}
]
