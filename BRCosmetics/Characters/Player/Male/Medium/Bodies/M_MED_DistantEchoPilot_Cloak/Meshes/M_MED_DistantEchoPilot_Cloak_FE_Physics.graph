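# Physics asset for the male medium DistantEchoPilot cloak, with capsule bodies for the jacket bones, spine, pelvis, legs and a small pelvis-attached box bone.
# Front, side and back chains hang from master bodies on the spine, using limited swing and tighter twist; the center back forks in two.
# Angularly free cross-constraints tie neighboring chains at selected rows, linking separate strips into a connected, cloth-like sheet.
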PhysicsAsset /BRCosmetics/Characters/Player/Male/Medium/Bodies/M_MED_DistantEchoPilot_Cloak/Meshes/M_MED_DistantEchoPilot_Cloak_FE_Physics
bodies {
 Jacket_Fr_l_01 [{shapeType=Capsule, length=7.5388, radius=6.4612, shapeName=Jacket_Fr_l_01_capsule, center=<6 5 4>, rotation=<rot 90 0 0>}]
 Jacket_Fr_l_02 [{shapeType=Capsule, length=7.6638, radius=6.3362, shapeName=Jacket_Fr_l_02_capsule, center=<6 5 4>, rotation=<rot 90 0 0>}]
 Jacket_Fr_l_03 [
  {shapeType=Capsule, length=7.6638, radius=6.3362, shapeName=Jacket_Fr_l_03_capsule, center=<6 5 4>, rotation=<rot 90 0 0>}
  {shapeType=Capsule, length=7, radius=6.5, shapeName=Jacket_Fr_l_03_capsule, center=<6 -3.0166 12.8871>, rotation=<rot 90 0 0>}
 ]
 Jacket_Fr_l_04 [{shapeType=Capsule, length=7.125, radius=6.875, shapeName=Jacket_Fr_l_04_capsule, center=<6 5 4>, rotation=<rot 90 0 0>}]
 Jacket_Fr_l_05 [{shapeType=Capsule, length=6.875, radius=7.125, shapeName=Jacket_Fr_l_05_capsule, center=<6 5 4>, rotation=<rot 90 0 0>}]
 Jacket_Fr_l_06 [{shapeType=Capsule, length=6.75, radius=7.25, shapeName=Jacket_Fr_l_06_capsule, center=<6 5 4>, rotation=<rot 90 0 0>}]
 Jacket_Fr_l_07 [{shapeType=Capsule, length=6.75, radius=7.25, shapeName=Jacket_Fr_l_07_capsule, center=<6 5 4>, rotation=<rot 90 0 0>}]
 Jacket_Fr_l_MSTR [{shapeType=Capsule, length=0.505, radius=0.505, shapeName=Jacket_Fr_l_MSTR_capsule, rotation=<rot 0 -90 0>}]
 Jacket_Fr_l_lookat [{shapeType=Capsule, length=0.505, radius=0.505, shapeName=Jacket_Fr_l_lookat_capsule, rotation=<rot 0 -90 0>}]
 Jacket_Fr_r_01 [{shapeType=Capsule, length=7.25, radius=6.75, shapeName=Jacket_Fr_r_01_capsule, center=<6 5 -4>, rotation=<rot 90 0 0>}]
 Jacket_Fr_r_02 [{shapeType=Capsule, length=6.2403, radius=7.7597, shapeName=Jacket_Fr_r_02_capsule, center=<6 5 -4>, rotation=<rot 90 0 0>}]
 Jacket_Fr_r_03 [{shapeType=Capsule, length=6.1142, radius=7.8858, shapeName=Jacket_Fr_r_03_capsule, center=<6 5 -4>, rotation=<rot 90 0 0>}]
 Jacket_Fr_r_04 [{shapeType=Capsule, length=6.1142, radius=7.8858, shapeName=Jacket_Fr_r_04_capsule, center=<6 5 -4>, rotation=<rot 90 0 0>}]
 Jacket_Fr_r_05 [{shapeType=Capsule, length=7.25, radius=6.75, shapeName=Jacket_Fr_r_05_capsule, center=<6 5 -4>, rotation=<rot 90 0 0>}]
 Jacket_Fr_r_06 [
  {shapeType=Capsule, length=5.625, radius=8.375, shapeName=Jacket_Fr_r_06_capsule, center=<6 5 -4>, rotation=<rot 90 0 0>}
  {shapeType=Capsule, length=8.875, radius=7.5, shapeName=Jacket_Fr_r_06_capsule, center=<7.1792 -5.0479 -15.6549>, rotation=<rot 90 0 -0.0002>}
 ]
 Jacket_Fr_r_07 [
  {shapeType=Capsule, length=5.625, radius=8.375, shapeName=Jacket_Fr_r_07_capsule, center=<6 5 -4>, rotation=<rot 90 0 0>}
  {
   shapeType=Capsule
   length=8.875
   radius=7.5
   shapeName=Jacket_Fr_r_07_capsule
   center=<7.1792 -5.0479 -15.6549>
   rotation=<rot 90 160.5287 160.5285>
  }
 ]
 Jacket_Fr_r_MSTR [{shapeType=Capsule, length=0.505, radius=0.505, shapeName=Jacket_Fr_r_MSTR_capsule, rotation=<rot 0 -90 0>}]
 Jacket_Fr_r_lookat [{shapeType=Capsule, length=0.505, radius=0.505, shapeName=Jacket_Fr_r_lookat_capsule, rotation=<rot 0 -90 0>}]
 calf_l [{shapeType=Capsule, length=40.5719, radius=8.3445, shapeName=calf_l_capsule, center=<-24.6622 0 2.1197>, rotation=<rot 90 0 0>}]
 calf_r [{shapeType=Capsule, length=41.4469, radius=7.4695, shapeName=calf_l_capsule, center=<24.3428 0 0>, rotation=<rot 90 0 0>}]
 dyn_box03 [{
   shapeType=Capsule
   length=2.6688
   radius=2.6924
   shapeName=dyn_box03_capsule
   center=<0.0982 0.0919 0.3643>
   rotation=<rot -0.9922 -90 2.8772>
  }]
 dyn_jacketBack_01 [{shapeType=Capsule, length=10.9178, radius=7.4534, shapeName=dyn_SkirtF03_c_capsule, center=<6.8844 -3 0>}]
 dyn_jacketBack_01_l [{shapeType=Capsule, length=12.7079, radius=5.6634, shapeName=dyn_SkirtE03_l_capsule, center=<6.8844 0 3>, rotation=<rot 90 0 0>}]
 dyn_jacketBack_01_r [{shapeType=Capsule, length=13.931, radius=5.742, shapeName=dyn_SkirtE03_r_capsule, center=<6.8844 0 3>, rotation=<rot 90 0 0>}]
 dyn_jacketBack_02 [{shapeType=Capsule, length=10.6529, radius=9.7779, shapeName=dyn_SkirtF04_c_capsule, center=<7 -8 0>}]
 dyn_jacketBack_02_l [{shapeType=Capsule, length=11.1149, radius=9.3159, shapeName=dyn_SkirtE04_l_capsule, center=<7 0 8>, rotation=<rot 0 0 90>}]
 dyn_jacketBack_02_r [{shapeType=Capsule, length=11.843, radius=8.5878, shapeName=dyn_SkirtE04_r_capsule, center=<7 0 8>, rotation=<rot 0 0 90>}]
 dyn_jacketBack_03_a [{shapeType=Capsule, length=5.4823, radius=10.6469, shapeName=dyn_SkirtF05_c_capsule, center=<7 -8 5>}]
 dyn_jacketBack_03_b [{shapeType=Capsule, length=5.4823, radius=10.6469, shapeName=dyn_SkirtF05_c_capsule, center=<7 -8 -5>}]
 dyn_jacketBack_03_l [{shapeType=Capsule, length=15.7493, radius=8.8396, shapeName=dyn_SkirtE05_l_capsule, center=<7 0 8>, rotation=<rot 0 0 90>}]
 dyn_jacketBack_03_r [{shapeType=Capsule, length=14.2639, radius=8.6115, shapeName=dyn_SkirtE05_r_capsule, center=<7 0 8>, rotation=<rot 0 0 90>}]
 dyn_jacketBack_04_a [{shapeType=Capsule, length=2.8985, radius=9.8211, shapeName=dyn_SkirtF05_c_capsule, center=<7 -6 5>}]
 dyn_jacketBack_04_b [{shapeType=Capsule, length=2.8985, radius=9.8211, shapeName=dyn_SkirtF05_c_capsule, center=<7 -6 -5>}]
 dyn_jacketBack_04_l [{shapeType=Capsule, length=16.0948, radius=13.1709, shapeName=dyn_SkirtE06_l_capsule, center=<7 0 8>, rotation=<rot -0.0001 0 85>}]
 dyn_jacketBack_04_r [{shapeType=Capsule, length=13.7578, radius=10.2584, shapeName=dyn_SkirtE06_r_capsule, center=<7 0 8>, rotation=<rot -0.0003 -0.0002 95>}]
 dyn_jacketBack_05_a [{shapeType=Capsule, length=2.5005, radius=10.6424, shapeName=dyn_SkirtF07_c_capsule, center=<7 -6 5>}]
 dyn_jacketBack_05_b [{shapeType=Capsule, length=2.5005, radius=10.6424, shapeName=dyn_SkirtF07_c_capsule, center=<7 -6 -5>}]
 dyn_jacketBack_05_l [{shapeType=Capsule, length=14.9978, radius=14.2679, shapeName=dyn_SkirtE07_l_capsule, center=<7 0 8>, rotation=<rot -0.0001 0 85>}]
 dyn_jacketBack_05_r [{shapeType=Capsule, length=12.6608, radius=11.3554, shapeName=dyn_SkirtE07_r_capsule, center=<7 0 8>, rotation=<rot -0.0003 -0.0002 95>}]
 dyn_jacketBack_MSTR [{shapeType=Capsule, length=0.505, radius=0.505, shapeName=dyn_jacketBack_MSTR_capsule, rotation=<rot 0 -90 0>}]
 dyn_jacketBack_mstr_l [{shapeType=Capsule, length=0.505, radius=0.505, shapeName=dyn_jacketBack_mstr_l_capsule, rotation=<rot 0 -90 0>}]
 dyn_jacketBack_mstr_r [{shapeType=Capsule, length=0.505, radius=0.505, shapeName=dyn_jacketBack_mstr_r_capsule, rotation=<rot 0 -90 0>}]
 dyn_jacketSide_01_l [{shapeType=Capsule, length=14.2955, radius=4.0757, shapeName=dyn_SkirtD03_r_capsule, center=<6.8844 0 3>, rotation=<rot 90 0 0>}]
 dyn_jacketSide_01_r [{shapeType=Capsule, length=13.8445, radius=4.5267, shapeName=dyn_SkirtD03_r_capsule, center=<6.8844 0 3>, rotation=<rot 90 0 0>}]
 dyn_jacketSide_02_l [{shapeType=Capsule, length=19.9873, radius=7.9367, shapeName=dyn_SkirtD07_r_capsule, center=<7 0 8>, rotation=<rot 0 0 90>}]
 dyn_jacketSide_02_r [{shapeType=Capsule, length=11.7166, radius=8.7142, shapeName=dyn_SkirtD04_r_capsule, center=<7 0 -8>, rotation=<rot 0 0 90>}]
 dyn_jacketSide_03_l [{shapeType=Capsule, length=19.9873, radius=7.9367, shapeName=dyn_SkirtD07_r_capsule, center=<7 0 8>, rotation=<rot 0 0 90>}]
 dyn_jacketSide_03_r [{shapeType=Capsule, length=16.7906, radius=7.7983, shapeName=dyn_SkirtD05_r_capsule, center=<7 0 -8>, rotation=<rot 0 0 90>}]
 dyn_jacketSide_04_l [{shapeType=Capsule, length=24.2373, radius=10.4367, shapeName=dyn_SkirtD07_r_capsule, center=<7 0 8>, rotation=<rot 0 0.0003 85>}]
 dyn_jacketSide_04_r [{shapeType=Capsule, length=26.5304, radius=9.6345, shapeName=dyn_SkirtD06_r_capsule, center=<7 0 -8>, rotation=<rot 0 0.0003 90>}]
 dyn_jacketSide_05_l [{shapeType=Capsule, length=23.3623, radius=11.3117, shapeName=dyn_SkirtD07_r_capsule, center=<7 0 8>, rotation=<rot 0 0.0003 85>}]
 dyn_jacketSide_05_r [{shapeType=Capsule, length=22.5727, radius=10.4762, shapeName=dyn_SkirtD07_r_capsule, center=<7 0 -8>, rotation=<rot 0 0.0003 90>}]
 dyn_jacketSide_mstr_l [{shapeType=Capsule, length=0.505, radius=0.505, shapeName=dyn_jacketSide_mstr_l_capsule, rotation=<rot 0 -90 0>}]
 dyn_jacketSide_mstr_r [{shapeType=Capsule, length=0.505, radius=0.505, shapeName=dyn_jacketSide_mstr_r_capsule, rotation=<rot 0 -90 0>}]
 foot_l [{
   shapeType=Capsule
   length=12.2345
   radius=12.4737
   shapeName=foot_l_capsule
   center=<-5.1272 7.6714 3.2443>
   rotation=<rot 0 -0.2319 -85.9858>
  }]
 foot_r [{
   shapeType=Capsule
   length=21.9616
   radius=10.1941
   shapeName=foot_r_capsule
   center=<4.5477 -13.0354 0.5092>
   rotation=<rot 0 0.4931 -85.4534>
  }]
 pelvis [
  {shapeType=Capsule, length=11.1836, radius=11.2351, shapeName=pelvis_capsule, center=<-2.1748 -0.7759 0>}
  {shapeType=Capsule, length=5.8642, radius=8.8898, shapeName=pelvis_capsule, center=<-2.2601 2.3375 4.4792>}
 ]
 spine_01 [{shapeType=Capsule, length=11.8459, radius=11.4436, shapeName=spine_01_capsule, center=<0 0.0442 0>}]
 spine_02 [{shapeType=Capsule, length=12.5397, radius=9.8644, shapeName=spine_02_capsule, center=<0 -0.0567 0>}]
 spine_03 [{shapeType=Capsule, length=14.4147, radius=11.6144, shapeName=spine_02_capsule, center=<0 -1.7658 0>}]
 spine_04 [{shapeType=Capsule, length=14.9147, radius=11.1144, shapeName=spine_02_capsule, center=<0 -0.5903 0>}]
 thigh_l [
  {shapeType=Capsule, length=28.0356, radius=10.1319, shapeName=thigh_l_capsule, center=<-19.8973 0 0.8625>, rotation=<rot 90 0 0>}
  {
   shapeType=Capsule
   length=23.9662
   radius=6.7156
   shapeName=thigh_l_capsule
   center=<-19.5427 14.8995 6.6268>
   rotation=<rot 86.1882 0 0.0001>
  }
 ]
 thigh_r [
  {shapeType=Capsule, length=30.1462, radius=9.501, shapeName=thigh_l_capsule, center=<19.8973 0 0>, rotation=<rot 90 0 0>}
  {
   shapeType=Capsule
   length=21.6263
   radius=9.0555
   shapeName=thigh_l_capsule
   center=<19.1461 -14.2641 -12.5801>
   rotation=<rot 86.1882 0 0.0001>
  }
 ]
 thigh_twist_01_r [{
   shapeType=Capsule
   length=12.9711
   radius=10.8936
   shapeName=thigh_twist_01_r_capsule
   center=<-6.0856 0.7306 -1.6047>
   rotation=<rot 90 179.5524 174.5525>
  }]
}
constraints [
 {bone1Name=spine_02, bone2Name=spine_01, swing1LimitDegrees=45, swing2LimitDegrees=45, twistLimitDegrees=45}
 {bone1Name=spine_03, bone2Name=spine_02, swing1LimitDegrees=45, swing2LimitDegrees=45, twistLimitDegrees=45}
 {bone1Name=spine_04, bone2Name=spine_03, swing1LimitDegrees=45, swing2LimitDegrees=45, twistLimitDegrees=45}
 {bone1Name=thigh_l, bone2Name=pelvis, swing1LimitDegrees=45, swing2LimitDegrees=45, twistLimitDegrees=45}
 {bone1Name=calf_l, bone2Name=thigh_l, swing1LimitDegrees=45, swing2LimitDegrees=45, twistLimitDegrees=45}
 {bone1Name=foot_l, bone2Name=calf_l, swing1LimitDegrees=45, swing2LimitDegrees=45, twistLimitDegrees=45}
 {bone1Name=thigh_r, bone2Name=pelvis, swing1LimitDegrees=45, swing2LimitDegrees=45, twistLimitDegrees=45}
 {bone1Name=calf_r, bone2Name=thigh_r, swing1LimitDegrees=45, swing2LimitDegrees=45, twistLimitDegrees=45}
 {bone1Name=foot_r, bone2Name=calf_r, swing1LimitDegrees=45, swing2LimitDegrees=45, twistLimitDegrees=45}
 {
  bone1Name=spine_01
  bone2Name=pelvis
  swing1LimitDegrees=45
  swing1Motion=Free
  swing2LimitDegrees=45
  swing2Motion=Free
  twistLimitDegrees=45
  twistMotion=Free
 }
 {
  bone1Name=dyn_jacketSide_mstr_l
  bone2Name=spine_01
  swing1LimitDegrees=45
  swing1Motion=Free
  swing2LimitDegrees=45
  swing2Motion=Free
  twistLimitDegrees=45
  twistMotion=Free
 }
 {bone1Name=dyn_jacketSide_01_l, bone2Name=dyn_jacketSide_mstr_l, swing1LimitDegrees=35, swing2LimitDegrees=35, twistLimitDegrees=5}
 {bone1Name=dyn_jacketSide_02_l, bone2Name=dyn_jacketSide_01_l, swing1LimitDegrees=35, swing2LimitDegrees=35, twistLimitDegrees=5}
 {bone1Name=dyn_jacketSide_03_l, bone2Name=dyn_jacketSide_02_l, swing1LimitDegrees=35, swing2LimitDegrees=35, twistLimitDegrees=5}
 {bone1Name=dyn_jacketSide_04_l, bone2Name=dyn_jacketSide_03_l, swing1LimitDegrees=35, swing2LimitDegrees=35, twistLimitDegrees=5}
 {bone1Name=dyn_jacketSide_05_l, bone2Name=dyn_jacketSide_04_l, swing1LimitDegrees=35, swing2LimitDegrees=35, twistLimitDegrees=5}
 {
  bone1Name=dyn_jacketSide_mstr_r
  bone2Name=spine_01
  swing1LimitDegrees=45
  swing1Motion=Free
  swing2LimitDegrees=45
  swing2Motion=Free
  twistLimitDegrees=45
  twistMotion=Free
 }
 {bone1Name=dyn_jacketSide_01_r, bone2Name=dyn_jacketSide_mstr_r, swing1LimitDegrees=35, swing2LimitDegrees=35, twistLimitDegrees=5}
 {bone1Name=dyn_jacketSide_02_r, bone2Name=dyn_jacketSide_01_r, swing1LimitDegrees=35, swing2LimitDegrees=35, twistLimitDegrees=5}
 {bone1Name=dyn_jacketSide_03_r, bone2Name=dyn_jacketSide_02_r, swing1LimitDegrees=35, swing2LimitDegrees=35, twistLimitDegrees=5}
 {bone1Name=dyn_jacketSide_04_r, bone2Name=dyn_jacketSide_03_r, swing1LimitDegrees=35, swing2LimitDegrees=35, twistLimitDegrees=5}
 {bone1Name=dyn_jacketSide_05_r, bone2Name=dyn_jacketSide_04_r, swing1LimitDegrees=35, swing2LimitDegrees=35, twistLimitDegrees=5}
 {
  bone1Name=dyn_jacketBack_MSTR
  bone2Name=spine_02
  swing1LimitDegrees=45
  swing1Motion=Free
  swing2LimitDegrees=45
  swing2Motion=Free
  twistLimitDegrees=45
  twistMotion=Free
 }
 {bone1Name=dyn_jacketBack_01, bone2Name=dyn_jacketBack_MSTR, swing1LimitDegrees=35, swing2LimitDegrees=35, twistLimitDegrees=15}
 {bone1Name=dyn_jacketBack_02, bone2Name=dyn_jacketBack_01, swing1LimitDegrees=35, swing2LimitDegrees=35, twistLimitDegrees=15}
 {bone1Name=dyn_jacketBack_03_a, bone2Name=dyn_jacketBack_02, swing1LimitDegrees=35, swing2LimitDegrees=35, twistLimitDegrees=15}
 {bone1Name=dyn_jacketBack_04_a, bone2Name=dyn_jacketBack_03_a, swing1LimitDegrees=35, swing2LimitDegrees=35, twistLimitDegrees=15}
 {bone1Name=dyn_jacketBack_05_a, bone2Name=dyn_jacketBack_04_a, swing1LimitDegrees=35, swing2LimitDegrees=35, twistLimitDegrees=15}
 {bone1Name=dyn_jacketBack_03_b, bone2Name=dyn_jacketBack_02, swing1LimitDegrees=35, swing2LimitDegrees=35, twistLimitDegrees=15}
 {bone1Name=dyn_jacketBack_04_b, bone2Name=dyn_jacketBack_03_b, swing1LimitDegrees=35, swing2LimitDegrees=35, twistLimitDegrees=15}
 {bone1Name=dyn_jacketBack_05_b, bone2Name=dyn_jacketBack_04_b, swing1LimitDegrees=35, swing2LimitDegrees=35, twistLimitDegrees=15}
 {
  bone1Name=dyn_jacketBack_mstr_l
  bone2Name=spine_02
  swing1LimitDegrees=45
  swing1Motion=Free
  swing2LimitDegrees=45
  swing2Motion=Free
  twistLimitDegrees=45
  twistMotion=Free
 }
 {bone1Name=dyn_jacketBack_01_l, bone2Name=dyn_jacketBack_mstr_l, swing1LimitDegrees=35, swing2LimitDegrees=35, twistLimitDegrees=15}
 {bone1Name=dyn_jacketBack_02_l, bone2Name=dyn_jacketBack_01_l, swing1LimitDegrees=35, swing2LimitDegrees=35, twistLimitDegrees=15}
 {bone1Name=dyn_jacketBack_03_l, bone2Name=dyn_jacketBack_02_l, swing1LimitDegrees=35, swing2LimitDegrees=35, twistLimitDegrees=15}
 {bone1Name=dyn_jacketBack_04_l, bone2Name=dyn_jacketBack_03_l, swing1LimitDegrees=35, swing2LimitDegrees=35, twistLimitDegrees=15}
 {bone1Name=dyn_jacketBack_05_l, bone2Name=dyn_jacketBack_04_l, swing1LimitDegrees=35, swing2LimitDegrees=35, twistLimitDegrees=15}
 {
  bone1Name=dyn_jacketBack_mstr_r
  bone2Name=spine_02
  swing1LimitDegrees=45
  swing1Motion=Free
  swing2LimitDegrees=45
  swing2Motion=Free
  twistLimitDegrees=45
  twistMotion=Free
 }
 {bone1Name=dyn_jacketBack_01_r, bone2Name=dyn_jacketBack_mstr_r, swing1LimitDegrees=35, swing2LimitDegrees=35, twistLimitDegrees=15}
 {bone1Name=dyn_jacketBack_02_r, bone2Name=dyn_jacketBack_01_r, swing1LimitDegrees=35, swing2LimitDegrees=35, twistLimitDegrees=15}
 {bone1Name=dyn_jacketBack_03_r, bone2Name=dyn_jacketBack_02_r, swing1LimitDegrees=35, swing2LimitDegrees=35, twistLimitDegrees=15}
 {bone1Name=dyn_jacketBack_04_r, bone2Name=dyn_jacketBack_03_r, swing1LimitDegrees=35, swing2LimitDegrees=35, twistLimitDegrees=15}
 {bone1Name=dyn_jacketBack_05_r, bone2Name=dyn_jacketBack_04_r, swing1LimitDegrees=35, swing2LimitDegrees=35, twistLimitDegrees=15}
 {
  bone1Name=Jacket_Fr_l_MSTR
  bone2Name=Jacket_Fr_l_lookat
  swing1LimitDegrees=45
  swing1Motion=Free
  swing2LimitDegrees=45
  swing2Motion=Free
  twistLimitDegrees=45
  twistMotion=Free
 }
 {
  bone1Name=Jacket_Fr_l_lookat
  bone2Name=spine_04
  swing1LimitDegrees=45
  swing1Motion=Free
  swing2LimitDegrees=45
  swing2Motion=Free
  twistLimitDegrees=45
  twistMotion=Free
 }
 {
  bone1Name=Jacket_Fr_r_lookat
  bone2Name=spine_04
  swing1LimitDegrees=45
  swing1Motion=Free
  swing2LimitDegrees=45
  swing2Motion=Free
  twistLimitDegrees=45
  twistMotion=Free
 }
 {
  bone1Name=Jacket_Fr_r_MSTR
  bone2Name=Jacket_Fr_r_lookat
  swing1LimitDegrees=45
  swing1Motion=Free
  swing2LimitDegrees=45
  swing2Motion=Free
  twistLimitDegrees=45
  twistMotion=Free
 }
 {
  bone1Name=dyn_jacketBack_03_b
  bone2Name=dyn_jacketBack_03_r
  swing1LimitDegrees=45
  swing1Motion=Free
  swing2LimitDegrees=45
  swing2Motion=Free
  twistLimitDegrees=45
  twistMotion=Free
 }
 {
  bone1Name=dyn_jacketBack_04_b
  bone2Name=dyn_jacketBack_04_r
  swing1LimitDegrees=45
  swing1Motion=Free
  swing2LimitDegrees=45
  swing2Motion=Free
  twistLimitDegrees=45
  twistMotion=Free
 }
 {
  bone1Name=dyn_jacketBack_04_a
  bone2Name=dyn_jacketBack_04_l
  swing1LimitDegrees=45
  swing1Motion=Free
  swing2LimitDegrees=45
  swing2Motion=Free
  twistLimitDegrees=45
  twistMotion=Free
 }
 {
  bone1Name=dyn_jacketBack_03_a
  bone2Name=dyn_jacketBack_03_l
  swing1LimitDegrees=45
  swing1Motion=Free
  swing2LimitDegrees=45
  swing2Motion=Free
  twistLimitDegrees=45
  twistMotion=Free
 }
 {
  bone1Name=dyn_jacketBack_05_b
  bone2Name=dyn_jacketBack_05_r
  swing1LimitDegrees=45
  swing1Motion=Free
  swing2LimitDegrees=45
  swing2Motion=Free
  twistLimitDegrees=45
  twistMotion=Free
 }
 {
  bone1Name=dyn_jacketBack_05_a
  bone2Name=dyn_jacketBack_05_l
  swing1LimitDegrees=45
  swing1Motion=Free
  swing2LimitDegrees=45
  swing2Motion=Free
  twistLimitDegrees=45
  twistMotion=Free
 }
 {
  bone1Name=dyn_jacketBack_04_l
  bone2Name=dyn_jacketSide_04_l
  swing1LimitDegrees=45
  swing1Motion=Free
  swing2LimitDegrees=45
  swing2Motion=Free
  twistLimitDegrees=45
  twistMotion=Free
 }
 {
  bone1Name=dyn_jacketBack_02_l
  bone2Name=dyn_jacketSide_02_l
  swing1LimitDegrees=45
  swing1Motion=Free
  swing2LimitDegrees=45
  swing2Motion=Free
  twistLimitDegrees=45
  twistMotion=Free
 }
 {
  bone1Name=dyn_jacketBack_04_r
  bone2Name=dyn_jacketSide_04_r
  swing1LimitDegrees=45
  swing1Motion=Free
  swing2LimitDegrees=45
  swing2Motion=Free
  twistLimitDegrees=45
  twistMotion=Free
 }
 {
  bone1Name=dyn_jacketBack_02_r
  bone2Name=dyn_jacketSide_02_r
  swing1LimitDegrees=45
  swing1Motion=Free
  swing2LimitDegrees=45
  swing2Motion=Free
  twistLimitDegrees=45
  twistMotion=Free
 }
 {bone1Name=Jacket_Fr_r_07, bone2Name=Jacket_Fr_r_06, swing1LimitDegrees=35, swing2LimitDegrees=35, twistLimitDegrees=15}
 {bone1Name=Jacket_Fr_r_01, bone2Name=Jacket_Fr_r_MSTR, swing1LimitDegrees=35, swing2LimitDegrees=35, twistLimitDegrees=15}
 {bone1Name=Jacket_Fr_r_02, bone2Name=Jacket_Fr_r_01, swing1LimitDegrees=35, swing2LimitDegrees=35, twistLimitDegrees=15}
 {bone1Name=Jacket_Fr_r_03, bone2Name=Jacket_Fr_r_02, swing1LimitDegrees=35, swing2LimitDegrees=35, twistLimitDegrees=15}
 {bone1Name=Jacket_Fr_r_04, bone2Name=Jacket_Fr_r_03, swing1LimitDegrees=35, swing2LimitDegrees=35, twistLimitDegrees=15}
 {bone1Name=Jacket_Fr_r_05, bone2Name=Jacket_Fr_r_04, swing1LimitDegrees=35, swing2LimitDegrees=35, twistLimitDegrees=15}
 {bone1Name=Jacket_Fr_r_06, bone2Name=Jacket_Fr_r_05, swing1LimitDegrees=35, swing2LimitDegrees=35, twistLimitDegrees=15}
 {bone1Name=Jacket_Fr_l_07, bone2Name=Jacket_Fr_l_06, swing1LimitDegrees=35, swing2LimitDegrees=35, twistLimitDegrees=15}
 {bone1Name=Jacket_Fr_l_01, bone2Name=Jacket_Fr_l_MSTR, swing1LimitDegrees=35, swing2LimitDegrees=35, twistLimitDegrees=15}
 {bone1Name=Jacket_Fr_l_02, bone2Name=Jacket_Fr_l_01, swing1LimitDegrees=35, swing2LimitDegrees=35, twistLimitDegrees=15}
 {bone1Name=Jacket_Fr_l_03, bone2Name=Jacket_Fr_l_02, swing1LimitDegrees=35, swing2LimitDegrees=35, twistLimitDegrees=15}
 {bone1Name=Jacket_Fr_l_04, bone2Name=Jacket_Fr_l_03, swing1LimitDegrees=35, swing2LimitDegrees=35, twistLimitDegrees=15}
 {bone1Name=Jacket_Fr_l_05, bone2Name=Jacket_Fr_l_04, swing1LimitDegrees=35, swing2LimitDegrees=35, twistLimitDegrees=15}
 {bone1Name=Jacket_Fr_l_06, bone2Name=Jacket_Fr_l_05, swing1LimitDegrees=35, swing2LimitDegrees=35, twistLimitDegrees=15}
 {
  bone1Name=Jacket_Fr_l_04
  bone2Name=dyn_jacketSide_02_l
  swing1LimitDegrees=45
  swing1Motion=Free
  swing2LimitDegrees=45
  swing2Motion=Free
  twistLimitDegrees=45
  twistMotion=Free
 }
 {
  bone1Name=Jacket_Fr_l_05
  bone2Name=dyn_jacketSide_03_l
  swing1LimitDegrees=45
  swing1Motion=Free
  swing2LimitDegrees=45
  swing2Motion=Free
  twistLimitDegrees=45
  twistMotion=Free
 }
 {
  bone1Name=Jacket_Fr_l_06
  bone2Name=dyn_jacketSide_04_l
  swing1LimitDegrees=45
  swing1Motion=Free
  swing2LimitDegrees=45
  swing2Motion=Free
  twistLimitDegrees=45
  twistMotion=Free
 }
 {
  bone1Name=Jacket_Fr_l_03
  bone2Name=dyn_jacketSide_01_l
  swing1LimitDegrees=45
  swing1Motion=Free
  swing2LimitDegrees=45
  swing2Motion=Free
  twistLimitDegrees=45
  twistMotion=Free
 }
 {
  bone1Name=Jacket_Fr_r_06
  bone2Name=dyn_jacketSide_04_r
  swing1LimitDegrees=45
  swing1Motion=Free
  swing2LimitDegrees=45
  swing2Motion=Free
  twistLimitDegrees=45
  twistMotion=Free
 }
 {
  bone1Name=Jacket_Fr_r_05
  bone2Name=dyn_jacketSide_03_r
  swing1LimitDegrees=45
  swing1Motion=Free
  swing2LimitDegrees=45
  swing2Motion=Free
  twistLimitDegrees=45
  twistMotion=Free
 }
 {
  bone1Name=Jacket_Fr_r_04
  bone2Name=dyn_jacketSide_02_r
  swing1LimitDegrees=45
  swing1Motion=Free
  swing2LimitDegrees=45
  swing2Motion=Free
  twistLimitDegrees=45
  twistMotion=Free
 }
 {
  bone1Name=Jacket_Fr_r_03
  bone2Name=dyn_jacketSide_01_r
  swing1LimitDegrees=45
  swing1Motion=Free
  swing2LimitDegrees=45
  swing2Motion=Free
  twistLimitDegrees=45
  twistMotion=Free
 }
 {
  bone1Name=dyn_jacketBack_03_b
  bone2Name=dyn_jacketBack_03_a
  swing1LimitDegrees=45
  swing1Motion=Free
  swing2LimitDegrees=45
  swing2Motion=Free
  twistLimitDegrees=45
  twistMotion=Free
 }
 {
  bone1Name=thigh_twist_01_r
  bone2Name=thigh_r
  swing1LimitDegrees=45
  swing1Motion=Free
  swing2LimitDegrees=45
  swing2Motion=Free
  twistLimitDegrees=45
  twistMotion=Free
 }
 {
  bone1Name=dyn_box03
  bone2Name=pelvis
  swing1LimitDegrees=45
  swing1Motion=Free
  swing2LimitDegrees=45
  swing2Motion=Free
  twistLimitDegrees=45
  twistMotion=Free
 }
]
