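# Animation Blueprint that drives the jacket and collar bones of the Distant Echo Pilot Cloak player cosmetic.
# It copies the parent mesh pose, then applies two Control Rigs and additive bone offsets for states like jumping, skydiving and sliding.
# Output switches on bRigidBodyEnabled between state-selected RigidBody simulations and a non-simulated fallback.
# The fallback, built from the 'prerigid' cached pose, layers thigh and crouch corrective rotations with movement-weighted Trail controllers.
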AnimBlueprint /BRCosmetics/Characters/Player/Male/Medium/Bodies/M_MED_DistantEchoPilot_Cloak/Meshes/M_MED_DistantEchoPilot_Cloak_AnimBP
target_skeleton=&/M_MED_DistantEchoPilot_Cloak_Skeleton
graphs [
 AnimGraph {nodes [
   Component To Local()
   n1 = "Transform (Modify) Bone - Bone: dyn_jacketSide_mstr_l"!(ComponentPose, bAlphaBoolEnabled:bool=True, Translation="0,0,0", Rotation="5,0,0.000", Scale="1,1,1") -> (Pose) {
    alphaInputType=Bool
    rotationMode=BMM_Additive
    rotationSpace=BCS_BoneSpace
    alphaBoolBlend {blendInTime=0.1, blendOption=Linear, blendOutTime=0.5}
    alphaScaleBiasClamp {bInterpResult=true, interpSpeedIncreasing=5}
    boneToModify {boneName=dyn_jacketSide_mstr_l}
    rotation {pitch=5}
   }
   "Transform (Modify) Bone - Bone: dyn_jacketSide_mstr_r"!(ComponentPose=@n1, bAlphaBoolEnabled:bool=True, Translation="0,0,0", Rotation="5,0,0.000", Scale="1,1,1") -> (Pose) {
    alphaInputType=Bool
    rotationMode=BMM_Additive
    rotationSpace=BCS_BoneSpace
    alphaBoolBlend {blendInTime=0.1, blendOption=Linear, blendOutTime=0.5}
    alphaScaleBiasClamp {bInterpResult=true, interpSpeedIncreasing=5}
    boneToModify {boneName=dyn_jacketSide_mstr_r}
    rotation {pitch=5}
   }
   n2 = "Transform (Modify) Bone - Bone: dyn_collar_fr"(Alpha=Get Arm R rot() -> (Arm R rot:double), Translation="0,0,0", Rotation="0,-50,25", Scale="1,1,1") {
    rotationMode=BMM_Additive
    rotationSpace=BCS_BoneSpace
    translationMode=BMM_Additive
    translationSpace=BCS_BoneSpace
    alphaBoolBlend {blendOption=Linear}
    boneToModify {boneName=dyn_collar_fr}
    rotation {roll=25, yaw=-50}
   }
   "Transform (Modify) Bone - Bone: dyn_jacket_collar_fr"(ComponentPose=@n2, Alpha=1, Translation="0,0,0", Rotation="15,-75,70", Scale="1,1,1") {
    rotationMode=BMM_Additive
    rotationSpace=BCS_BoneSpace
    translationMode=BMM_Additive
    translationSpace=BCS_BoneSpace
    alphaBoolBlend {blendOption=Linear}
    boneToModify {boneName=dyn_jacket_collar_fr}
    rotation=<rot 15 -75 70>
   }
   n3 = "Transform (Modify) Bone - Bone: dyn_box03"!(
    ComponentPose=Local To Component(LocalPose=Use cached pose 'prerigid'())
    bAlphaBoolEnabled:bool=True
    Translation="-0.5,0,0"
    Rotation="0,0,0"
    Scale="0.7,1.000,1.000"
   ) -> (Pose) {
    alphaInputType=Bool
    rotationMode=BMM_Additive
    rotationSpace=BCS_BoneSpace
    scaleMode=BMM_Additive
    scaleSpace=BCS_BoneSpace
    translationMode=BMM_Additive
    translationSpace=BCS_BoneSpace
    alphaBoolBlend {blendInTime=0.1, blendOption=Linear, blendOutTime=0.5}
    alphaScaleBiasClamp {bInterpResult=true, interpSpeedIncreasing=5}
    boneToModify {boneName=dyn_box03}
    scale {x=0.7}
    translation {x=-0.5}
   }
   n4 = Get GravityOverride()
   n5 = RigidBody!(ComponentPose=@n3, Alpha, OverrideWorldGravity:Vector=@n4, ExternalForce) -> (Pose) {
    bClampLinearTranslationLimitToRefPose=true
    bOverrideWorldGravity=true
    overridePhysicsAsset=&/M_MED_DistantEchoPilot_Cloak_FE_Physics
    simulationSpace=BaseBoneSpace
    alphaBoolBlend {blendOption=Linear}
    baseBoneRef {boneName=root}
    componentAppliedLinearAccClamp=<3500 3500 3500>
    componentLinearAccScale=<0.3 0.3 0.3>
    componentLinearVelScale=<0.3 0.3 0.3>
    overrideWorldGravity {z=-980}
   }
   n6 = RigidBody(ComponentPose=Local To Component(LocalPose=Use cached pose 'prerigid'()), OverrideWorldGravity=Get GravityOverride()) {
    bClampLinearTranslationLimitToRefPose=true
    bOverrideWorldGravity=true
    overridePhysicsAsset=&/M_MED_DistantEchoPilot_Cloak_Physics
    simulationSpace=BaseBoneSpace
    alphaBoolBlend {blendOption=Linear}
    baseBoneRef {boneName=root}
    componentAppliedLinearAccClamp=<3500 3500 3500>
    componentLinearAccScale=<0.3 0.25 0.45>
    componentLinearVelScale=<0.3 0.25 0.45>
    overrideWorldGravity {z=-980}
   }
   n7 = Get BwdMovement_2() -> (BwdMovement_2:double)
   n8 = "Transform (Modify) Bone - Bone: dyn_jacketBack_MSTR"(ComponentPose=@n6, Alpha=@n7, Translation="0,0,0", Rotation="0,-10,0.000", Scale="1,1,1") {
    rotationMode=BMM_Additive
    rotationSpace=BCS_BoneSpace
    alphaBoolBlend {blendOption=Linear}
    alphaScaleBiasClamp {bInterpResult=true, interpSpeedIncreasing=30}
    boneToModify {boneName=dyn_jacketBack_MSTR}
    rotation {yaw=-10}
   }
   n9 = "Transform (Modify) Bone - Bone: dyn_jacketBack_mstr_l"(ComponentPose=@n8, Alpha=@n7, Translation="0,0,0", Rotation="7,0,0.000", Scale="1,1,1") {
    rotationMode=BMM_Additive
    rotationSpace=BCS_BoneSpace
    alphaBoolBlend {blendOption=Linear}
    alphaScaleBiasClamp {bInterpResult=true, interpSpeedIncreasing=30}
    boneToModify {boneName=dyn_jacketBack_mstr_l}
    rotation {pitch=7}
   }
   n10 = "Transform (Modify) Bone - Bone: dyn_jacketBack_mstr_r"(ComponentPose=@n9, Alpha=@n7, Translation="0,0,0", Rotation="7,0,0.000", Scale="1,1,1") {
    rotationMode=BMM_Additive
    rotationSpace=BCS_BoneSpace
    alphaBoolBlend {blendOption=Linear}
    alphaScaleBiasClamp {bInterpResult=true, interpSpeedIncreasing=30}
    boneToModify {boneName=dyn_jacketBack_mstr_r}
    rotation {pitch=7}
   }
   n11 = RigidBody(ComponentPose=Local To Component(LocalPose=Use cached pose 'prerigid'()), OverrideWorldGravity=Get GravityOverride()) {
    bClampLinearTranslationLimitToRefPose=true
    bOverrideWorldGravity=true
    overridePhysicsAsset=&/M_MED_DistantEchoPilot_Cloak_emote_Physics
    simulationSpace=BaseBoneSpace
    alphaBoolBlend {blendOption=Linear}
    baseBoneRef {boneName=pelvis}
    componentAppliedLinearAccClamp=<3500 3500 3500>
    componentLinearAccScale=<0.25 0.25 0.25>
    componentLinearVelScale=<0.25 0.25 0.25>
    overrideWorldGravity {z=-980}
   }
   n12 = "Transform (Modify) Bone - Bone: dyn_jacketBack_MSTR"(ComponentPose=@n11, Alpha=1, Translation="0,0,0", Rotation="0,-10,0.000", Scale="1,1,1") {
    rotationMode=BMM_Additive
    rotationSpace=BCS_BoneSpace
    alphaBoolBlend {blendOption=Linear}
    alphaScaleBiasClamp {bInterpResult=true, interpSpeedIncreasing=30}
    boneToModify {boneName=dyn_jacketBack_MSTR}
    rotation {yaw=-10}
   }
   n13 = "Transform (Modify) Bone - Bone: dyn_jacketBack_mstr_l"(ComponentPose=@n12, Alpha=1, Translation="0,0,0", Rotation="7,0,0.000", Scale="1,1,1") {
    rotationMode=BMM_Additive
    rotationSpace=BCS_BoneSpace
    alphaBoolBlend {blendOption=Linear}
    alphaScaleBiasClamp {bInterpResult=true, interpSpeedIncreasing=30}
    boneToModify {boneName=dyn_jacketBack_mstr_l}
    rotation {pitch=7}
   }
   n14 = "Transform (Modify) Bone - Bone: dyn_jacketBack_mstr_r"(ComponentPose=@n13, Alpha=1, Translation="0,0,0", Rotation="7,0,0.000", Scale="1,1,1") {
    rotationMode=BMM_Additive
    rotationSpace=BCS_BoneSpace
    alphaBoolBlend {blendOption=Linear}
    alphaScaleBiasClamp {bInterpResult=true, interpSpeedIncreasing=30}
    boneToModify {boneName=dyn_jacketBack_mstr_r}
    rotation {pitch=7}
   }
   n15 = Blend Poses by bool(
    BlendPose_0=Component To Local(ComponentPose=@n14)
    BlendPose_1=Blend Poses by bool(BlendPose_0=Component To Local(ComponentPose=@n5), BlendPose_1=Component To Local(ComponentPose=@n10), bActiveValue=Get bIsInFrontEnd()) {blendType=Linear}
    BlendTime_1=0.3
    bActiveValue=Get Emote_bool() -> (Emote_bool:bool)
   ) {blendType=Linear}
   n16 = RigidBody(ComponentPose=Local To Component(LocalPose=Use cached pose 'prerigid'()), OverrideWorldGravity=@n4) {
    bClampLinearTranslationLimitToRefPose=true
    bOverrideWorldGravity=true
    overridePhysicsAsset=&M_MED_DistantEchoPilot_Cloak_Physics
    simulationSpace=BaseBoneSpace
    alphaBoolBlend {blendOption=Linear}
    baseBoneRef {boneName=pelvis}
    componentAppliedLinearAccClamp=<3500 3500 3500>
    componentLinearAccScale=<0.3 0.3 0.3>
    componentLinearVelScale=<0.3 0.3 0.3>
    overrideWorldGravity {z=-980}
   }
   n17 = Component To Local(ComponentPose=@n16)
   n18 = Get bIsSkyDiving()
   n19 = Get thigh_shape_lt_l() -> (thigh_shape_lt_l:double)
   n20 = "Transform (Modify) Bone - Bone: dyn_jacketBack_03_b"(
    ComponentPose=Local To Component(LocalPose=Blend Poses by bool(BlendPose_0=@n17, BlendPose_1=@n15, bActiveValue=@n18) {blendType=Linear})
    Alpha=@n19
    Translation="0,0,0"
    Rotation="0,-20,0"
    Scale="1,1,1"
   ) {
    rotationMode=BMM_Additive
    rotationSpace=BCS_BoneSpace
    scaleMode=BMM_Additive
    scaleSpace=BCS_BoneSpace
    translationMode=BMM_Additive
    translationSpace=BCS_BoneSpace
    alphaBoolBlend {blendInTime=0.1, blendOption=Linear, blendOutTime=0.5}
    boneToModify {boneName=dyn_jacketBack_03_b}
    rotation {yaw=-20}
   }
   n21 = "Transform (Modify) Bone - Bone: dyn_jacketBack_03_r"(ComponentPose=@n20, Alpha=@n19, Translation="0,0,0", Rotation="25,0,0", Scale="1,1,1") {
    rotationMode=BMM_Additive
    rotationSpace=BCS_BoneSpace
    scaleMode=BMM_Additive
    scaleSpace=BCS_BoneSpace
    translationMode=BMM_Additive
    translationSpace=BCS_BoneSpace
    alphaBoolBlend {blendInTime=0.1, blendOption=Linear, blendOutTime=0.5}
    boneToModify {boneName=dyn_jacketBack_03_r}
    rotation {pitch=25}
   }
   n22 = Get bIsMoving2D()
   n23 = "Transform (Modify) Bone - Bone: dyn_jacketBack_MSTR"!(ComponentPose=@n21, bAlphaBoolEnabled:bool=@n22, Translation="0,0,0", Rotation="0,-5,0.000", Scale="1,1,1") -> (Pose) {
    alphaInputType=Bool
    rotationMode=BMM_Additive
    rotationSpace=BCS_BoneSpace
    alphaBoolBlend {blendInTime=0.1, blendOption=Linear, blendOutTime=0.5}
    alphaScaleBiasClamp {bInterpResult=true, interpSpeedIncreasing=5}
    boneToModify {boneName=dyn_jacketBack_MSTR}
    rotation {yaw=-5}
   }
   n24 = "Transform (Modify) Bone - Bone: dyn_jacketBack_mstr_l"!(ComponentPose=@n23, bAlphaBoolEnabled:bool=@n22, Translation="0,0,0", Rotation="5,0,0.000", Scale="1,1,1") -> (Pose) {
    alphaInputType=Bool
    rotationMode=BMM_Additive
    rotationSpace=BCS_BoneSpace
    alphaBoolBlend {blendInTime=0.1, blendOption=Linear, blendOutTime=0.5}
    alphaScaleBiasClamp {bInterpResult=true, interpSpeedIncreasing=5}
    boneToModify {boneName=dyn_jacketBack_mstr_l}
    rotation {pitch=5}
   }
   n25 = "Transform (Modify) Bone - Bone: dyn_jacketBack_mstr_r"!(ComponentPose=@n24, bAlphaBoolEnabled:bool=@n22, Translation="0,0,0", Rotation="5,0,0.000", Scale="1,1,1") -> (Pose) {
    alphaInputType=Bool
    rotationMode=BMM_Additive
    rotationSpace=BCS_BoneSpace
    alphaBoolBlend {blendInTime=0.1, blendOption=Linear, blendOutTime=0.5}
    alphaScaleBiasClamp {bInterpResult=true, interpSpeedIncreasing=5}
    boneToModify {boneName=dyn_jacketBack_mstr_r}
    rotation {pitch=5}
   }
   n26 = Get bIsMoving2D()
   n27 = "Transform (Modify) Bone - Bone: dyn_jacketBack_MSTR"(
    ComponentPose=Local To Component(LocalPose=Use cached pose 'prerigid'())
    bAlphaBoolEnabled=@n26
    Translation="0,0,0"
    Rotation="0,-10,0.000"
    Scale="1,1,1"
   ) {
    alphaInputType=Bool
    rotationMode=BMM_Additive
    rotationSpace=BCS_BoneSpace
    alphaBoolBlend {blendInTime=0.1, blendOption=Linear, blendOutTime=0.5}
    alphaScaleBiasClamp {bInterpResult=true, interpSpeedIncreasing=5}
    boneToModify {boneName=dyn_jacketBack_MSTR}
    rotation {yaw=-10}
   }
   n28 = "Transform (Modify) Bone - Bone: dyn_jacketBack_mstr_l"(ComponentPose=@n27, bAlphaBoolEnabled=@n26, Translation="0,0,0", Rotation="10,0,0.000", Scale="1,1,1") {
    alphaInputType=Bool
    rotationMode=BMM_Additive
    rotationSpace=BCS_BoneSpace
    alphaBoolBlend {blendInTime=0.1, blendOption=Linear, blendOutTime=0.5}
    alphaScaleBiasClamp {bInterpResult=true, interpSpeedIncreasing=5}
    boneToModify {boneName=dyn_jacketBack_mstr_l}
    rotation {pitch=10}
   }
   n29 = "Transform (Modify) Bone - Bone: dyn_jacketBack_mstr_r"(ComponentPose=@n28, bAlphaBoolEnabled=@n26, Translation="0,0,0", Rotation="10,0,0.000", Scale="1,1,1") {
    alphaInputType=Bool
    rotationMode=BMM_Additive
    rotationSpace=BCS_BoneSpace
    alphaBoolBlend {blendInTime=0.1, blendOption=Linear, blendOutTime=0.5}
    alphaScaleBiasClamp {bInterpResult=true, interpSpeedIncreasing=5}
    boneToModify {boneName=dyn_jacketBack_mstr_r}
    rotation {pitch=10}
   }
   n30 = Get thigh_rot_inr_l() -> (thigh_rot_inr_l:double)
   n31 = "Transform (Modify) Bone - Bone: dyn_jacketSide_mstr_l"!(ComponentPose=@n29, Alpha:float=@n30, Translation="0,0,0", Rotation="15,0,0.000", Scale="1,1,1") -> (Pose) {
    rotationMode=BMM_Additive
    rotationSpace=BCS_BoneSpace
    alphaBoolBlend {blendOption=Linear}
    boneToModify {boneName=dyn_jacketSide_mstr_l}
    rotation {pitch=15}
   }
   n32 = Get thigh_rot_inr_r() -> (thigh_rot_inr_r:double)
   n33 = "Transform (Modify) Bone - Bone: dyn_jacketSide_mstr_r"!(ComponentPose=@n31, Alpha:float=@n32, Translation="0,0,0", Rotation="15,0,0.000", Scale="1,1,1") -> (Pose) {
    rotationMode=BMM_Additive
    rotationSpace=BCS_BoneSpace
    alphaBoolBlend {blendOption=Linear}
    alphaScaleBiasClamp {interpSpeedIncreasing=5}
    boneToModify {boneName=dyn_jacketSide_mstr_r}
    rotation {pitch=15}
   }
   n34 = "Transform (Modify) Bone - Bone: Jacket_Fr_l_MSTR"(ComponentPose=@n33, Alpha=@n30, Translation="0,0,0", Rotation="25,0,0.000", Scale="1,1,1") {
    rotationMode=BMM_Additive
    rotationSpace=BCS_BoneSpace
    alphaBoolBlend {blendOption=Linear}
    boneToModify {boneName=Jacket_Fr_l_MSTR}
    rotation {pitch=25}
   }
   n35 = "Transform (Modify) Bone - Bone: Jacket_Fr_r_MSTR"(ComponentPose=@n34, Alpha=@n32, Translation="0,0,0", Rotation="-20,0,0.000", Scale="1,1,1") {
    rotationMode=BMM_Additive
    rotationSpace=BCS_BoneSpace
    alphaBoolBlend {blendOption=Linear}
    alphaScaleBiasClamp {interpSpeedIncreasing=5}
    boneToModify {boneName=Jacket_Fr_r_MSTR}
    rotation {pitch=-20}
   }
   n36 = "Transform (Modify) Bone - Bone: dyn_jacketBack_mstr_l"!(ComponentPose=@n35, Alpha:float=@n30, Translation="0,0,0", Rotation="10,0,0.000", Scale="1,1,1") -> (Pose) {
    rotationMode=BMM_Additive
    rotationSpace=BCS_BoneSpace
    alphaBoolBlend {blendOption=Linear}
    alphaScaleBiasClamp {interpSpeedIncreasing=5}
    boneToModify {boneName=dyn_jacketBack_mstr_l}
    rotation {pitch=10}
   }
   n37 = "Transform (Modify) Bone - Bone: dyn_jacketBack_mstr_r"!(ComponentPose=@n36, Alpha:float=@n32, Translation="0,0,0", Rotation="10,0,0.000", Scale="1,1,1") -> (Pose) {
    rotationMode=BMM_Additive
    rotationSpace=BCS_BoneSpace
    alphaBoolBlend {blendOption=Linear}
    boneToModify {boneName=dyn_jacketBack_mstr_r}
    rotation {pitch=10}
   }
   n38 = Get thigh_rot_otr_l() -> (thigh_rot_otr_l:double)
   n39 = "Transform (Modify) Bone - Bone: dyn_jacketSide_mstr_l"(ComponentPose=@n37, Alpha=@n38, Translation="0,0,0", Rotation="15,0,0.000", Scale="1,1,1") {
    rotationMode=BMM_Additive
    rotationSpace=BCS_BoneSpace
    alphaBoolBlend {blendOption=Linear}
    boneToModify {boneName=dyn_jacketSide_mstr_l}
    rotation {pitch=15}
   }
   n40 = Get thigh_rot_otr_r() -> (thigh_rot_otr_r:double)
   n41 = "Transform (Modify) Bone - Bone: dyn_jacketSide_mstr_r"(ComponentPose=@n39, Alpha=@n40, Translation="0,0,0", Rotation="15,0,0.000", Scale="1,1,1") {
    rotationMode=BMM_Additive
    rotationSpace=BCS_BoneSpace
    alphaBoolBlend {blendOption=Linear}
    alphaScaleBiasClamp {interpSpeedIncreasing=5}
    boneToModify {boneName=dyn_jacketSide_mstr_r}
    rotation {pitch=15}
   }
   n42 = "Transform (Modify) Bone - Bone: Jacket_Fr_l_MSTR"(ComponentPose=@n41, Alpha=@n38, Translation="0,0,0", Rotation="25,15,0", Scale="1,1,1") {
    rotationMode=BMM_Additive
    rotationSpace=BCS_BoneSpace
    translationMode=BMM_Additive
    translationSpace=BCS_BoneSpace
    alphaBoolBlend {blendOption=Linear}
    boneToModify {boneName=Jacket_Fr_l_MSTR}
    rotation {pitch=25, yaw=15}
   }
   n43 = "Transform (Modify) Bone - Bone: Jacket_Fr_r_MSTR"(ComponentPose=@n42, Alpha=@n40, Translation="0,0,0", Rotation="-20,15,0", Scale="1,1,1") {
    rotationMode=BMM_Additive
    rotationSpace=BCS_BoneSpace
    translationMode=BMM_Additive
    translationSpace=BCS_BoneSpace
    alphaBoolBlend {blendOption=Linear}
    alphaScaleBiasClamp {interpSpeedIncreasing=5}
    boneToModify {boneName=Jacket_Fr_r_MSTR}
    rotation {pitch=-20, yaw=15}
   }
   n44 = "Transform (Modify) Bone - Bone: dyn_jacketBack_mstr_l"(ComponentPose=@n43, Alpha=@n38, Translation="0,0,0", Rotation="10,0,0.000", Scale="1,1,1") {
    rotationMode=BMM_Additive
    rotationSpace=BCS_BoneSpace
    alphaBoolBlend {blendOption=Linear}
    alphaScaleBiasClamp {interpSpeedIncreasing=5}
    boneToModify {boneName=dyn_jacketBack_mstr_l}
    rotation {pitch=10}
   }
   n45 = "Transform (Modify) Bone - Bone: dyn_jacketBack_mstr_r"(ComponentPose=@n44, Alpha=@n40, Translation="0,0,0", Rotation="10,0,0.000", Scale="1,1,1") {
    rotationMode=BMM_Additive
    rotationSpace=BCS_BoneSpace
    alphaBoolBlend {blendOption=Linear}
    boneToModify {boneName=dyn_jacketBack_mstr_r}
    rotation {pitch=10}
   }
   n46 = Get still_crouch() -> (still_crouch:bool)
   n47 = "Transform (Modify) Bone - Bone: Jacket_Fr_l_05"!(ComponentPose=@n45, bAlphaBoolEnabled:bool=@n46, Translation="0,0,0", Rotation="20,0,0", Scale="1,1,1") -> (Pose) {
    alphaInputType=Bool
    rotationMode=BMM_Additive
    rotationSpace=BCS_BoneSpace
    scaleMode=BMM_Additive
    scaleSpace=BCS_BoneSpace
    translationMode=BMM_Additive
    translationSpace=BCS_BoneSpace
    alphaBoolBlend {blendInTime=0.1, blendOption=Linear, blendOutTime=0.5}
    boneToModify {boneName=Jacket_Fr_l_05}
    rotation {pitch=20}
   }
   n48 = "Transform (Modify) Bone - Bone: Jacket_Fr_l_06"!(ComponentPose=@n47, bAlphaBoolEnabled:bool=@n46, Translation="0,0,0", Rotation="20,0,0.015", Scale="1,1,1") -> (Pose) {
    alphaInputType=Bool
    rotationMode=BMM_Additive
    rotationSpace=BCS_BoneSpace
    scaleMode=BMM_Additive
    scaleSpace=BCS_BoneSpace
    translationMode=BMM_Additive
    translationSpace=BCS_BoneSpace
    alphaBoolBlend {blendInTime=0.1, blendOption=Linear, blendOutTime=0.5}
    boneToModify {boneName=Jacket_Fr_l_06}
    rotation {pitch=20, roll=0.015}
   }
   n49 = "Transform (Modify) Bone - Bone: dyn_jacketBack_02"!(ComponentPose=@n48, bAlphaBoolEnabled:bool=@n46, Translation="0,0,0", Rotation="-15,-30,0", Scale="1,1,1") -> (Pose) {
    alphaInputType=Bool
    rotationMode=BMM_Additive
    rotationSpace=BCS_BoneSpace
    scaleMode=BMM_Additive
    scaleSpace=BCS_BoneSpace
    translationMode=BMM_Additive
    translationSpace=BCS_BoneSpace
    alphaBoolBlend {blendInTime=0.1, blendOption=Linear, blendOutTime=0.5}
    boneToModify {boneName=dyn_jacketBack_02}
    rotation {pitch=-15, yaw=-30}
   }
   n50 = "Transform (Modify) Bone - Bone: dyn_jacketBack_02_r"!(ComponentPose=@n49, bAlphaBoolEnabled:bool=@n46, Translation="0,0,0", Rotation="35,-0.000,0.000", Scale="1,1,1") -> (Pose) {
    alphaInputType=Bool
    rotationMode=BMM_Additive
    rotationSpace=BCS_BoneSpace
    scaleMode=BMM_Additive
    scaleSpace=BCS_BoneSpace
    translationMode=BMM_Additive
    translationSpace=BCS_BoneSpace
    alphaBoolBlend {blendInTime=0.1, blendOption=Linear, blendOutTime=0.5}
    boneToModify {boneName=dyn_jacketBack_02_r}
    rotation {pitch=35}
   }
   n51 = "Transform (Modify) Bone - Bone: dyn_jacketBack_03_r"!(ComponentPose=@n50, bAlphaBoolEnabled:bool=@n46, Translation="0,0,0", Rotation="15,0,0", Scale="1,1,1") -> (Pose) {
    alphaInputType=Bool
    rotationMode=BMM_Additive
    rotationSpace=BCS_BoneSpace
    scaleMode=BMM_Additive
    scaleSpace=BCS_BoneSpace
    translationMode=BMM_Additive
    translationSpace=BCS_BoneSpace
    alphaBoolBlend {blendInTime=0.1, blendOption=Linear, blendOutTime=0.5}
    boneToModify {boneName=dyn_jacketBack_03_r}
    rotation {pitch=15}
   }
   n52 = "Transform (Modify) Bone - Bone: dyn_jacketSide_02_r"!(ComponentPose=@n51, bAlphaBoolEnabled:bool=@n46, Translation="0,0,0", Rotation="20.000,15,0.000", Scale="1,1,1") -> (Pose) {
    alphaInputType=Bool
    rotationMode=BMM_Additive
    rotationSpace=BCS_BoneSpace
    scaleMode=BMM_Additive
    scaleSpace=BCS_BoneSpace
    translationMode=BMM_Additive
    translationSpace=BCS_BoneSpace
    alphaBoolBlend {blendInTime=0.1, blendOption=Linear, blendOutTime=0.5}
    boneToModify {boneName=dyn_jacketSide_02_r}
    rotation {pitch=20, yaw=15}
   }
   n53 = "Transform (Modify) Bone - Bone: dyn_jacketBack_03_l"!(ComponentPose=@n52, bAlphaBoolEnabled:bool=@n46, Translation="0,0,0", Rotation="15,0,0", Scale="1,1,1") -> (Pose) {
    alphaInputType=Bool
    rotationMode=BMM_Additive
    rotationSpace=BCS_BoneSpace
    scaleMode=BMM_Additive
    scaleSpace=BCS_BoneSpace
    translationMode=BMM_Additive
    translationSpace=BCS_BoneSpace
    alphaBoolBlend {blendInTime=0.1, blendOption=Linear, blendOutTime=0.5}
    boneToModify {boneName=dyn_jacketBack_03_l}
    rotation {pitch=15}
   }
   n54 = "Transform (Modify) Bone - Bone: dyn_jacketBack_04_l"!(ComponentPose=@n53, bAlphaBoolEnabled:bool=@n46, Translation="0,0,0", Rotation="15,0,0", Scale="1,1,1") -> (Pose) {
    alphaInputType=Bool
    rotationMode=BMM_Additive
    rotationSpace=BCS_BoneSpace
    scaleMode=BMM_Additive
    scaleSpace=BCS_BoneSpace
    translationMode=BMM_Additive
    translationSpace=BCS_BoneSpace
    alphaBoolBlend {blendInTime=0.1, blendOption=Linear, blendOutTime=0.5}
    boneToModify {boneName=dyn_jacketBack_04_l}
    rotation {pitch=15}
   }
   n55 = "Transform (Modify) Bone - Bone: dyn_jacketSide_03_l"!(ComponentPose=@n54, bAlphaBoolEnabled:bool=@n46, Translation="0,0,0", Rotation="20,0,0", Scale="1,1,1") -> (Pose) {
    alphaInputType=Bool
    rotationMode=BMM_Additive
    rotationSpace=BCS_BoneSpace
    scaleMode=BMM_Additive
    scaleSpace=BCS_BoneSpace
    translationMode=BMM_Additive
    translationSpace=BCS_BoneSpace
    alphaBoolBlend {blendInTime=0.1, blendOption=Linear, blendOutTime=0.5}
    boneToModify {boneName=dyn_jacketSide_03_l}
    rotation {pitch=20}
   }
   n56 = "Transform (Modify) Bone - Bone: dyn_jacketSide_03_r"!(ComponentPose=@n55, bAlphaBoolEnabled:bool=@n46, Translation="0,0,0", Rotation="-15,0,0", Scale="1,1,1") -> (Pose) {
    alphaInputType=Bool
    rotationMode=BMM_Additive
    rotationSpace=BCS_BoneSpace
    scaleMode=BMM_Additive
    scaleSpace=BCS_BoneSpace
    translationMode=BMM_Additive
    translationSpace=BCS_BoneSpace
    alphaBoolBlend {blendInTime=0.1, blendOption=Linear, blendOutTime=0.5}
    boneToModify {boneName=dyn_jacketSide_03_r}
    rotation {pitch=-15}
   }
   n57 = "Transform (Modify) Bone - Bone: Jacket_Fr_r_03"!(ComponentPose=@n56, bAlphaBoolEnabled:bool=@n46, Translation="0,0,0", Rotation="-20,0,0", Scale="1,1,1") -> (Pose) {
    alphaInputType=Bool
    rotationMode=BMM_Additive
    rotationSpace=BCS_BoneSpace
    scaleMode=BMM_Additive
    scaleSpace=BCS_BoneSpace
    translationMode=BMM_Additive
    translationSpace=BCS_BoneSpace
    alphaBoolBlend {blendInTime=0.1, blendOption=Linear, blendOutTime=0.5}
    boneToModify {boneName=Jacket_Fr_r_03}
    rotation {pitch=-20}
   }
   n58 = "Transform (Modify) Bone - Bone: Jacket_Fr_r_05"!(ComponentPose=@n57, bAlphaBoolEnabled:bool=@n46, Translation="0,0,0", Rotation="-20,0,0.015", Scale="1,1,1") -> (Pose) {
    alphaInputType=Bool
    rotationMode=BMM_Additive
    rotationSpace=BCS_BoneSpace
    scaleMode=BMM_Additive
    scaleSpace=BCS_BoneSpace
    translationMode=BMM_Additive
    translationSpace=BCS_BoneSpace
    alphaBoolBlend {blendInTime=0.1, blendOption=Linear, blendOutTime=0.5}
    boneToModify {boneName=Jacket_Fr_r_05}
    rotation {pitch=-20, roll=0.015}
   }
   n59 = "Transform (Modify) Bone - Bone: Jacket_Fr_r_06"!(ComponentPose=@n58, bAlphaBoolEnabled:bool=@n46, Translation="0,0,0", Rotation="-20,0,0.015", Scale="1,1,1") -> (Pose) {
    alphaInputType=Bool
    rotationMode=BMM_Additive
    rotationSpace=BCS_BoneSpace
    scaleMode=BMM_Additive
    scaleSpace=BCS_BoneSpace
    translationMode=BMM_Additive
    translationSpace=BCS_BoneSpace
    alphaBoolBlend {blendInTime=0.1, blendOption=Linear, blendOutTime=0.5}
    boneToModify {boneName=Jacket_Fr_r_06}
    rotation {pitch=-20, roll=0.015}
   }
   n60 = Get crouch move nosim() -> (crouch move nosim:bool)
   n61 = "Transform (Modify) Bone - Bone: dyn_jacketBack_MSTR"(ComponentPose=@n59, bAlphaBoolEnabled=@n60, Translation="0,0,0", Rotation="0,20,0.000", Scale="1,1,1") {
    alphaInputType=Bool
    rotationMode=BMM_Additive
    rotationSpace=BCS_BoneSpace
    alphaBoolBlend {blendInTime=0.1, blendOption=Linear, blendOutTime=0.1}
    alphaScaleBiasClamp {bInterpResult=true, interpSpeedIncreasing=5}
    boneToModify {boneName=dyn_jacketBack_MSTR}
    rotation {yaw=20}
   }
   n62 = "Transform (Modify) Bone - Bone: dyn_jacketBack_mstr_l"!(ComponentPose=@n61, bAlphaBoolEnabled:bool=@n60, Translation="0,0,0", Rotation="-15,0,0.000", Scale="1,1,1") -> (Pose) {
    alphaInputType=Bool
    rotationMode=BMM_Additive
    rotationSpace=BCS_BoneSpace
    alphaBoolBlend {blendInTime=0.1, blendOption=Linear, blendOutTime=0.1}
    alphaScaleBiasClamp {bInterpResult=true, interpSpeedIncreasing=5}
    boneToModify {boneName=dyn_jacketBack_mstr_l}
    rotation {pitch=-15}
   }
   n63 = "Transform (Modify) Bone - Bone: dyn_jacketBack_mstr_r"!(ComponentPose=@n62, bAlphaBoolEnabled:bool=@n60, Translation="0,0,0", Rotation="-15,0,0.000", Scale="1,1,1") -> (Pose) {
    alphaInputType=Bool
    rotationMode=BMM_Additive
    rotationSpace=BCS_BoneSpace
    alphaBoolBlend {blendInTime=0.1, blendOption=Linear, blendOutTime=0.1}
    alphaScaleBiasClamp {bInterpResult=true, interpSpeedIncreasing=5}
    boneToModify {boneName=dyn_jacketBack_mstr_r}
    rotation {pitch=-15}
   }
   n64 = Get BwdMovement() -> (BwdMovement:double)
   n65 = Trail controller - Bone: dyn_jacketBack_05_r!(ComponentPose=@n63, Alpha=@n64, FakeVelocity:Vector="0,0,0") -> (Pose) {
    bActorSpaceFakeVel=true
    bShowTrailLocation=true
    chainLength=4
    debugLifeTime=0.1
    maxDeltaTime=0.0333
    relaxationSpeedScale=3
    alphaScaleBiasClamp {bInterpResult=true}
    baseJoint {boneName=root}
    planarLimits [
     {drivingBone {boneName=root}, planeTransform {location=<0 0 10>, rotation=<rot 0 0 0>, scale=<1 1 1>}}
     {drivingBone {boneName=spine_03}, planeTransform {location=<0 -20 0>, rotation=<rot 0 15 -89.9999>, scale=<1 1 1>}}
    ]
    rotationLimits [{limitMax=<180 180 180>}, {limitMax=<180 180 180>}, {limitMax=<180 180 180>}, {limitMax=<180 180 180>}]
    rotationOffsets [<0 0 0>, <0 0 0>, <0 0 0>, <0 0 0>]
    trailBone {boneName=dyn_jacketBack_05_r}
    trailRelaxationSpeed {editorCurveData {keys [{arriveTangent=-72.0739, interpMode=RCIM_Cubic, leaveTangent=-72.0739, tangentMode=RCTM_User, time=0, value=35}, <at 1 5 RCIM_Cubic>]}}
   }
   n66 = Trail controller - Bone: dyn_jacketBack_05_l!(ComponentPose=@n65, Alpha=@n64, FakeVelocity:Vector="0,0,0") -> (Pose) {
    bActorSpaceFakeVel=true
    bShowTrailLocation=true
    chainLength=4
    debugLifeTime=0.1
    maxDeltaTime=0.0333
    relaxationSpeedScale=3
    alphaScaleBiasClamp {bInterpResult=true}
    baseJoint {boneName=root}
    planarLimits [
     {drivingBone {boneName=root}, planeTransform {location=<0 0 10>, rotation=<rot 0 0 0>, scale=<1 1 1>}}
     {drivingBone {boneName=spine_03}, planeTransform {location=<0 -20 0>, rotation=<rot 0 15 -89.9999>, scale=<1 1 1>}}
    ]
    rotationLimits [{limitMax=<180 180 180>}, {limitMax=<180 180 180>}, {limitMax=<180 180 180>}, {limitMax=<180 180 180>}]
    rotationOffsets [<0 0 0>, <0 0 0>, <0 0 0>, <0 0 0>]
    trailBone {boneName=dyn_jacketBack_05_l}
    trailRelaxationSpeed {editorCurveData {keys [{arriveTangent=-72.0739, interpMode=RCIM_Cubic, leaveTangent=-72.0739, tangentMode=RCTM_User, time=0, value=35}, <at 1 5 RCIM_Cubic>]}}
   }
   n67 = Get BwdMovement()
   n68 = Trail controller - Bone: dyn_jacketBack_02(ComponentPose=@n66, Alpha=@n67) {
    bUsePlanarLimit=true
    debugLifeTime=0.1
    maxDeltaTime=0.0333
    alphaScaleBiasClamp {bInterpResult=true}
    rotationLimits [{limitMax=<180 180 180>}, {limitMax=<180 180 180>}]
    rotationOffsets [<0 0 0>, <0 0 0>]
    trailBone {boneName=dyn_jacketBack_02}
    trailRelaxationSpeed {editorCurveData {keys [<at 0 15 RCIM_Linear>,<at 0.5289 13.1838 RCIM_Linear>,<at 1 6 RCIM_Linear>]}}
   }
   n69 = Trail controller - Bone: dyn_jacketBack_05_a(ComponentPose=@n68, Alpha=@n67) {
    bUsePlanarLimit=true
    debugLifeTime=0.1
    maxDeltaTime=0.333
    alphaScaleBiasClamp {bInterpResult=true}
    rotationLimits [{limitMax=<180 180 180>}, {limitMax=<180 180 180>}]
    rotationOffsets [<0 0 0>, <0 0 0>]
    trailBone {boneName=dyn_jacketBack_05_a}
    trailRelaxationSpeed {editorCurveData {keys [<at -0.0134 19.3841 RCIM_Linear>,<at 0.5079 12.8986 RCIM_Linear>,<at 0.9955 9.6044 RCIM_Linear>]}}
   }
   n70 = Trail controller - Bone: dyn_jacketBack_05_b(ComponentPose=@n69, Alpha=@n67) {
    bUsePlanarLimit=true
    debugLifeTime=0.1
    maxDeltaTime=0.333
    alphaScaleBiasClamp {bInterpResult=true}
    rotationLimits [{limitMax=<180 180 180>}, {limitMax=<180 180 180>}]
    rotationOffsets [<0 0 0>, <0 0 0>]
    trailBone {boneName=dyn_jacketBack_05_b}
    trailRelaxationSpeed {editorCurveData {keys [<at -0.0134 19.3841 RCIM_Linear>,<at 0.5079 12.8986 RCIM_Linear>,<at 0.9955 9.6044 RCIM_Linear>]}}
   }
   n71 = Trail controller - Bone: dyn_jacketSide_05_r!(ComponentPose=@n70, Alpha=Get LeftMovement() -> (LeftMovement:float), FakeVelocity:Vector="0,0,0") -> (Pose) {
    bActorSpaceFakeVel=true
    bShowTrailLocation=true
    chainLength=4
    debugLifeTime=0.1
    maxDeltaTime=0.0333
    relaxationSpeedScale=3
    alphaScaleBiasClamp {bInterpResult=true}
    baseJoint {boneName=root}
    planarLimits [
     {drivingBone {boneName=root}, planeTransform {location=<0 0 10>, rotation=<rot 0 0 0>, scale=<1 1 1>}}
     {drivingBone {boneName=spine_03}, planeTransform {location=<0 -20 0>, rotation=<rot 0 15 -89.9999>, scale=<1 1 1>}}
    ]
    rotationLimits [{limitMax=<180 180 180>}, {limitMax=<180 180 180>}, {limitMax=<180 180 180>}, {limitMax=<180 180 180>}]
    rotationOffsets [<0 0 0>, <0 0 0>, <0 0 0>, <0 0 0>]
    trailBone {boneName=dyn_jacketSide_05_r}
    trailRelaxationSpeed {editorCurveData {keys [{arriveTangent=-72.0739, interpMode=RCIM_Cubic, leaveTangent=-72.0739, tangentMode=RCTM_User, time=0, value=35}, <at 1 5 RCIM_Cubic>]}}
   }
   n72 = Trail controller - Bone: dyn_jacketSide_05_l!(ComponentPose=@n71, Alpha=Get RightMovement() -> (RightMovement:float), FakeVelocity:Vector="0,0,0") -> (Pose) {
    bActorSpaceFakeVel=true
    bShowTrailLocation=true
    chainLength=4
    debugLifeTime=0.1
    maxDeltaTime=0.0333
    relaxationSpeedScale=3
    alphaScaleBiasClamp {bInterpResult=true}
    baseJoint {boneName=root}
    planarLimits [
     {drivingBone {boneName=root}, planeTransform {location=<0 0 10>, rotation=<rot 0 0 0>, scale=<1 1 1>}}
     {drivingBone {boneName=spine_03}, planeTransform {location=<0 -20 0>, rotation=<rot 0 15 -89.9999>, scale=<1 1 1>}}
    ]
    rotationLimits [{limitMax=<180 180 180>}, {limitMax=<180 180 180>}, {limitMax=<180 180 180>}, {limitMax=<180 180 180>}]
    rotationOffsets [<0 0 0>, <0 0 0>, <0 0 0>, <0 0 0>]
    trailBone {boneName=dyn_jacketSide_05_l}
    trailRelaxationSpeed {editorCurveData {keys [{arriveTangent=-72.0739, interpMode=RCIM_Cubic, leaveTangent=-72.0739, tangentMode=RCTM_User, time=0, value=35}, <at 1 5 RCIM_Cubic>]}}
   }
   Output Pose(Result=Blend Poses by bool(BlendPose_0=Component To Local(ComponentPose=@n25), BlendPose_1=Component To Local(ComponentPose=@n72), bActiveValue=Get bRigidBodyEnabled()) {blendType=Linear}) {layerGroup=None}
   n73 = "Transform (Modify) Bone - Bone: dyn_Jacket_fr_TGT"!(
    ComponentPose=Local To Component(LocalPose=Copy Pose From Mesh() {bUseAttachedParent=true})
    bAlphaBoolEnabled:bool="Get skydive "() -> ("skydive ":bool)
    Translation="-0.000,-85,0"
    Rotation="0,0,0"
    Scale="1,1,1"
   ) -> (Pose) {
    alphaInputType=Bool
    rotationMode=BMM_Additive
    rotationSpace=BCS_BoneSpace
    translationMode=BMM_Additive
    translationSpace=BCS_BoneSpace
    alphaBoolBlend {blendInTime=0.1, blendOption=Linear, blendOutTime=0.1}
    alphaScaleBiasClamp {bInterpResult=true, interpSpeedIncreasing=5}
    boneToModify {boneName=dyn_Jacket_fr_TGT}
    translation {y=-85}
   }
   n74 = Control Rig(Source=Component To Local(ComponentPose=@n73)) {
    alphaBoolBlend {blendOption=Linear}
    controlRigAssetReference {blueprintRigClass=&/M_MED_Distant_Echo_Pilot_Cloak_DeformRig_ControlRig.M_MED_Distant_Echo_Pilot_Cloak_DeformRig_ControlRig_C}
   }
   n75 = Control Rig!(
    Source=@n74
    Alpha
    IsCrouchingCtrl:bool=Get bIsCrouching()
    Pickaxe:bool=Get melee_crouch() -> (melee_crouch:bool)
    emote:double=Get Emote() -> (Emote:double)
   ) -> (Pose) {
    alphaBoolBlend {blendOption=Linear}
    controlRigAssetReference {blueprintRigClass=&/M_MED_DistantEchoPilot_Cloak_CtrlRig.M_MED_DistantEchoPilot_Cloak_CtrlRig_C}
   }
   n76 = Get jumping() -> (jumping:double)
   n77 = "Transform (Modify) Bone - Bone: dyn_jacketBack_MSTR"!(ComponentPose=Local To Component(LocalPose=@n75), Alpha:float=@n76, Translation="0,0,0", Rotation="0,-35,0.000", Scale="1,1,1") -> (Pose) {
    rotationMode=BMM_Additive
    rotationSpace=BCS_BoneSpace
    alphaBoolBlend {blendOption=Linear}
    alphaScaleBiasClamp {bInterpResult=true, interpSpeedIncreasing=5}
    boneToModify {boneName=dyn_jacketBack_MSTR}
    rotation {yaw=-35}
   }
   n78 = "Transform (Modify) Bone - Bone: dyn_jacketBack_mstr_l"!(ComponentPose=@n77, Alpha:float=@n76, Translation="0,0,0", Rotation="30,10,0.000", Scale="1,1,1") -> (Pose) {
    rotationMode=BMM_Additive
    rotationSpace=BCS_BoneSpace
    alphaBoolBlend {blendOption=Linear}
    alphaScaleBiasClamp {bInterpResult=true, interpSpeedIncreasing=5}
    boneToModify {boneName=dyn_jacketBack_mstr_l}
    rotation {pitch=30, yaw=10}
   }
   n79 = "Transform (Modify) Bone - Bone: dyn_jacketBack_mstr_r"!(ComponentPose=@n78, Alpha:float=@n76, Translation="0,0,0", Rotation="30,-10,0.000", Scale="1,1,1") -> (Pose) {
    rotationMode=BMM_Additive
    rotationSpace=BCS_BoneSpace
    alphaBoolBlend {blendOption=Linear}
    alphaScaleBiasClamp {bInterpResult=true, interpSpeedIncreasing=5}
    boneToModify {boneName=dyn_jacketBack_mstr_r}
    rotation {pitch=30, yaw=-10}
   }
   n80 = "Transform (Modify) Bone - Bone: dyn_jacketSide_mstr_l"(ComponentPose=@n79, Alpha=@n76, Translation="0,0,0", Rotation="15,-15,0.000", Scale="1,1,1") {
    rotationMode=BMM_Additive
    rotationSpace=BCS_BoneSpace
    alphaBoolBlend {blendOption=Linear}
    alphaScaleBiasClamp {bInterpResult=true, interpSpeedIncreasing=5}
    boneToModify {boneName=dyn_jacketSide_mstr_l}
    rotation {pitch=15, yaw=-15}
   }
   n81 = "Transform (Modify) Bone - Bone: dyn_jacketSide_mstr_r"(ComponentPose=@n80, Alpha=@n76, Translation="0,0,0", Rotation="15,15,0.000", Scale="1,1,1") {
    rotationMode=BMM_Additive
    rotationSpace=BCS_BoneSpace
    alphaBoolBlend {blendOption=Linear}
    alphaScaleBiasClamp {bInterpResult=true, interpSpeedIncreasing=5}
    boneToModify {boneName=dyn_jacketSide_mstr_r}
    rotation {pitch=15, yaw=15}
   }
   n82 = "Transform (Modify) Bone - Bone: FloorCOL"(ComponentPose=@n81, Alpha=@n76, Translation="0.000,-0.000,-100", Rotation="0,0,0", Scale="1,1,1") {
    rotationMode=BMM_Additive
    rotationSpace=BCS_BoneSpace
    translationMode=BMM_Additive
    translationSpace=BCS_BoneSpace
    alphaBoolBlend {blendOption=Linear}
    alphaScaleBiasClamp {bInterpResult=true, interpSpeedIncreasing=5}
    boneToModify {boneName=FloorCOL}
    translation {z=-100}
   }
   n83 = Get bIsSkyDiving()
   n84 = "Transform (Modify) Bone - Bone: dyn_jacketBack_MSTR"!(ComponentPose=@n82, bAlphaBoolEnabled:bool=@n83, Translation="0,0,0", Rotation="0,-15,0.000", Scale="1,1,1") -> (Pose) {
    alphaInputType=Bool
    rotationMode=BMM_Additive
    rotationSpace=BCS_BoneSpace
    alphaBoolBlend {blendOption=Linear}
    alphaScaleBiasClamp {bInterpResult=true, interpSpeedIncreasing=5}
    boneToModify {boneName=dyn_jacketBack_MSTR}
    rotation {yaw=-15}
   }
   n85 = "Transform (Modify) Bone - Bone: dyn_jacketBack_mstr_l"!(ComponentPose=@n84, bAlphaBoolEnabled:bool=@n83, Translation="0,0,0", Rotation="15,10,0.000", Scale="1,1,1") -> (Pose) {
    alphaInputType=Bool
    rotationMode=BMM_Additive
    rotationSpace=BCS_BoneSpace
    alphaBoolBlend {blendOption=Linear}
    alphaScaleBiasClamp {bInterpResult=true, interpSpeedIncreasing=5}
    boneToModify {boneName=dyn_jacketBack_mstr_l}
    rotation {pitch=15, yaw=10}
   }
   n86 = "Transform (Modify) Bone - Bone: dyn_jacketBack_mstr_r"!(ComponentPose=@n85, bAlphaBoolEnabled:bool=@n83, Translation="0,0,0", Rotation="15,-10,0.000", Scale="1,1,1") -> (Pose) {
    alphaInputType=Bool
    rotationMode=BMM_Additive
    rotationSpace=BCS_BoneSpace
    alphaBoolBlend {blendOption=Linear}
    alphaScaleBiasClamp {bInterpResult=true, interpSpeedIncreasing=5}
    boneToModify {boneName=dyn_jacketBack_mstr_r}
    rotation {pitch=15, yaw=-10}
   }
   n87 = "Transform (Modify) Bone - Bone: dyn_jacketSide_mstr_l"!(ComponentPose=@n86, bAlphaBoolEnabled:bool=@n83, Translation="0,0,0", Rotation="25,0,0.000", Scale="1,1,1") -> (Pose) {
    alphaInputType=Bool
    rotationMode=BMM_Additive
    rotationSpace=BCS_BoneSpace
    alphaBoolBlend {blendOption=Linear}
    alphaScaleBiasClamp {bInterpResult=true, interpSpeedIncreasing=5}
    boneToModify {boneName=dyn_jacketSide_mstr_l}
    rotation {pitch=25}
   }
   n88 = "Transform (Modify) Bone - Bone: dyn_jacketSide_mstr_r"!(ComponentPose=@n87, bAlphaBoolEnabled:bool=@n83, Translation="0,0,0", Rotation="25,0,0.000", Scale="1,1,1") -> (Pose) {
    alphaInputType=Bool
    rotationMode=BMM_Additive
    rotationSpace=BCS_BoneSpace
    alphaBoolBlend {blendOption=Linear}
    alphaScaleBiasClamp {bInterpResult=true, interpSpeedIncreasing=5}
    boneToModify {boneName=dyn_jacketSide_mstr_r}
    rotation {pitch=25}
   }
   n89 = Get bIsSliding()
   n90 = "Transform (Modify) Bone - Bone: dyn_jacketBack_MSTR"(ComponentPose=@n88, bAlphaBoolEnabled=@n89, Translation="0,0,0", Rotation="0,-60,0.000", Scale="1,1,1") {
    alphaInputType=Bool
    rotationMode=BMM_Additive
    rotationSpace=BCS_BoneSpace
    alphaBoolBlend {blendInTime=0.1, blendOption=Linear, blendOutTime=0.5}
    alphaScaleBiasClamp {bInterpResult=true, interpSpeedIncreasing=5}
    boneToModify {boneName=dyn_jacketBack_MSTR}
    rotation {yaw=-60}
   }
   n91 = "Transform (Modify) Bone - Bone: dyn_jacketBack_mstr_l"(ComponentPose=@n90, bAlphaBoolEnabled=@n89, Translation="0,0,0", Rotation="45,5,0.000", Scale="1,1,1") {
    alphaInputType=Bool
    rotationMode=BMM_Additive
    rotationSpace=BCS_BoneSpace
    alphaBoolBlend {blendInTime=0.1, blendOption=Linear, blendOutTime=0.5}
    alphaScaleBiasClamp {bInterpResult=true, interpSpeedIncreasing=5}
    boneToModify {boneName=dyn_jacketBack_mstr_l}
    rotation {pitch=45, yaw=5}
   }
   n92 = "Transform (Modify) Bone - Bone: dyn_jacketBack_mstr_r"(ComponentPose=@n91, bAlphaBoolEnabled=@n89, Translation="0,0,0", Rotation="45,-5,0.000", Scale="1,1,1") {
    alphaInputType=Bool
    rotationMode=BMM_Additive
    rotationSpace=BCS_BoneSpace
    alphaBoolBlend {blendInTime=0.1, blendOption=Linear, blendOutTime=0.5}
    alphaScaleBiasClamp {bInterpResult=true, interpSpeedIncreasing=5}
    boneToModify {boneName=dyn_jacketBack_mstr_r}
    rotation {pitch=45, yaw=-5}
   }
   n93 = Get bIsInFrontEnd()
   n94 = "Transform (Modify) Bone - Bone: dyn_collar_bk"!(ComponentPose=@n92, bAlphaBoolEnabled:bool=@n93, Translation="1,0.000,-0.000", Rotation="-25,0.000,-0.000", Scale="1,1,1") -> (Pose) {
    alphaInputType=Bool
    rotationMode=BMM_Additive
    rotationSpace=BCS_BoneSpace
    translationMode=BMM_Additive
    translationSpace=BCS_BoneSpace
    alphaBoolBlend {blendInTime=0.1, blendOption=Linear, blendOutTime=0.5}
    alphaScaleBiasClamp {bInterpResult=true, interpSpeedIncreasing=5}
    boneToModify {boneName=dyn_collar_bk}
    rotation {pitch=-25}
    translation {x=1}
   }
   n95 = "Transform (Modify) Bone - Bone: dyn_collar_l"!(ComponentPose=@n94, bAlphaBoolEnabled:bool=@n93, Translation="1,0,0", Rotation="-35,0.000,-0.000", Scale="1,1,1") -> (Pose) {
    alphaInputType=Bool
    rotationMode=BMM_Additive
    rotationSpace=BCS_BoneSpace
    translationMode=BMM_Additive
    translationSpace=BCS_BoneSpace
    alphaBoolBlend {blendInTime=0.1, blendOption=Linear, blendOutTime=0.5}
    alphaScaleBiasClamp {bInterpResult=true, interpSpeedIncreasing=5}
    boneToModify {boneName=dyn_collar_l}
    rotation {pitch=-35}
    translation {x=1}
   }
   n96 = "Transform (Modify) Bone - Bone: attach_backpack"!(ComponentPose=@n95, bAlphaBoolEnabled:bool=Get bIsSkyDiving(), Translation="0,-1,0", Rotation="-0.000,0.000,5", Scale="1,1,1") -> (Pose) {
    alphaInputType=Bool
    rotationMode=BMM_Additive
    rotationSpace=BCS_BoneSpace
    translationMode=BMM_Additive
    translationSpace=BCS_BoneSpace
    alphaBoolBlend {blendOption=Linear}
    alphaScaleBiasClamp {bInterpResult=true, interpSpeedIncreasing=5}
    boneToModify {boneName=attach_backpack}
    rotation {roll=5}
    translation {y=-1}
   }
   n97 = "Transform (Modify) Bone - Bone: FloorCOL"!(ComponentPose=@n96, bAlphaBoolEnabled:bool=Get bIsPlayingEmote(), Translation="0,0,-100", Rotation="0,0,0", Scale="1,1,1") -> (Pose) {
    alphaInputType=Bool
    rotationMode=BMM_Additive
    rotationSpace=BCS_BoneSpace
    translationMode=BMM_Additive
    translationSpace=BCS_BoneSpace
    alphaBoolBlend {blendOption=Linear}
    alphaScaleBiasClamp {bInterpResult=true, interpSpeedIncreasing=5}
    boneToModify {boneName=FloorCOL}
    translation {z=-100}
   }
   n98 = "Constraint - dyn_box01 (PPPPPPPP)"!(
    ComponentPose=@n97
    Alpha=1
    ConstraintWeights_0=0.4924
    ConstraintWeights_1:float=0.0013
    ConstraintWeights_2:float=0.0024
    ConstraintWeights_3:float=0.0008
    ConstraintWeights_4:float=0.224
    ConstraintWeights_5:float=0.2096
    ConstraintWeights_6:float=0.0505
    ConstraintWeights_7:float=0.0191
   ) -> (Pose) {
    alphaBoolBlend {blendOption=Linear}
    boneToModify {boneName=dyn_box01}
    constraintSetup [
     {targetBone {boneName=pelvis}}
     {targetBone {boneName=spine_01}}
     {targetBone {boneName=spine_02}}
     {targetBone {boneName=spine_03}}
     {targetBone {boneName=thigh_twist_01_l}}
     {targetBone {boneName=deform_groin_l}}
     {targetBone {boneName=thigh_twist_01_r}}
     {targetBone {boneName=deform_groin_r}}
    ]
    constraintWeights [1,1,1,1,1,1,1,1]
   }
   n99 = "Constraint - dyn_box02 (PPPPPPPP)"!(
    ComponentPose=@n98
    Alpha=1
    ConstraintWeights_0=0.4882
    ConstraintWeights_1:float=0.0012
    ConstraintWeights_2:float=0.0025
    ConstraintWeights_3:float=0.0007
    ConstraintWeights_4:float=0.0088
    ConstraintWeights_5:float=0.003
    ConstraintWeights_6:float=0.2643
    ConstraintWeights_7:float=0.2313
   ) -> (Pose) {
    alphaBoolBlend {blendOption=Linear}
    boneToModify {boneName=dyn_box02}
    constraintSetup [
     {targetBone {boneName=pelvis}}
     {targetBone {boneName=spine_01}}
     {targetBone {boneName=spine_02}}
     {targetBone {boneName=spine_03}}
     {targetBone {boneName=thigh_twist_01_l}}
     {targetBone {boneName=deform_groin_l}}
     {targetBone {boneName=thigh_twist_01_r}}
     {targetBone {boneName=deform_groin_r}}
    ]
    constraintWeights [1,1,1,1,1,1,1,1]
   }
   n100 = "Constraint - dyn_box03 (PPPPPPPP)"!(
    ComponentPose=@n99
    Alpha=1
    ConstraintWeights_0=0.5437
    ConstraintWeights_1:float=0.0082
    ConstraintWeights_2:float=0.0097
    ConstraintWeights_3:float=0.0051
    ConstraintWeights_4:float=0.0001
    ConstraintWeights_5:float=0.275
    ConstraintWeights_6:float=0.0068
    ConstraintWeights_7:float=0.1515
   ) -> (Pose) {
    alphaBoolBlend {blendOption=Linear}
    boneToModify {boneName=dyn_box03}
    constraintSetup [
     {targetBone {boneName=pelvis}}
     {targetBone {boneName=spine_01}}
     {targetBone {boneName=spine_02}}
     {targetBone {boneName=spine_03}}
     {targetBone {boneName=thigh_twist_01_l}}
     {targetBone {boneName=thigh_twist_01_r}}
     {targetBone {boneName=deform_glute_r}}
     {targetBone {boneName=deform_groin_r}}
    ]
    constraintWeights [1,1,1,1,1,1,1,1]
   }
   n101 = "Constraint - dyn_tube (PPPPPPPP)"!(
    ComponentPose=@n100
    Alpha=1
    ConstraintWeights_0=0.5294
    ConstraintWeights_1:float=0.003
    ConstraintWeights_2:float=0.0046
    ConstraintWeights_3:float=0.0017
    ConstraintWeights_4:float=0.0516
    ConstraintWeights_5:float=0.0201
    ConstraintWeights_6:float=0.2076
    ConstraintWeights_7:float=0.182
   ) -> (Pose) {
    alphaBoolBlend {blendOption=Linear}
    boneToModify {boneName=dyn_tube}
    constraintSetup [
     {targetBone {boneName=pelvis}}
     {targetBone {boneName=spine_01}}
     {targetBone {boneName=spine_02}}
     {targetBone {boneName=spine_03}}
     {targetBone {boneName=thigh_twist_01_l}}
     {targetBone {boneName=deform_groin_l}}
     {targetBone {boneName=thigh_twist_01_r}}
     {targetBone {boneName=deform_groin_r}}
    ]
    constraintWeights [1,1,1,1,1,1,1,1]
   }
   Save cached pose 'prerigid'(Pose=Component To Local(ComponentPose=@n101)) {cacheName=prerigid}
   Get bIsPlayingEmote()
  ]}
 EventGraph {nodes [
   n1 = Event BlueprintUpdateAnimation()
   > Set DeltaTimeLocal(DeltaTimeLocal:double=@n1.DeltaTimeX) -> (Output_Get:double)
   > n2 = Sequence!(execute) -> (then_0, then_1, then_2:Exec, then_3:Exec, then_4:Exec, then_5:Exec, then_6:Exec, then_7:Exec)
   n3 = Get bIsMoving2D()
   > n4 = Branch(Condition=AND Boolean(A=Get bIsCrouching(), B=NOT Boolean(A=@n3)))
   > Set still_crouch(still_crouch:bool=true) -> (Output_Get:bool)
   Set still_crouch(execute=@n4.else, still_crouch=false)
   n5 = Gravity Override State Function DEPRECATED(
    execute=@n2.then_1
    CharacterPartAnimInstance=Self-Reference()
    GameTimeInSeconds=GetGameTimeInSeconds()
    GravityOverrideParams=Get GravityOverrideParams() -> (GravityOverrideParams:Gravity Override Params Struct Structure)
   )
   > Set GravityOverride(GravityOverride=@n5)
   n6 = Branch(execute=@n2.then_2, Condition=OR Boolean(A=Get bIsSkydiveDiveMode(), B=Get bIsSkydiveFloating()))
   > "Set skydive "("skydive ":bool=true) -> (Output_Get:bool)
   "Set skydive "(execute=@n6.else, "skydive "=false)
   n7 = Get Skel_Mesh()
   Set thigh_rot_otr_l(
    execute=@n2.then_3
    thigh_rot_otr_l:double=GetAlphaFromReferenceTransformDelta(SkeletalMeshComponent=@n7, BoneName=thigh_l, BaseName=pelvis, DeltaComponentType=EulerY, InputRange=(X=0,Y=30), TwistAxis="1,0,0")
   ) -> (Output_Get:double)
   > Set thigh_rot_otr_r(thigh_rot_otr_r:double=GetAlphaFromReferenceTransformDelta(SkeletalMeshComponent=@n7, BoneName=thigh_r, BaseName=pelvis, DeltaComponentType=EulerY, InputRange=(X=0,Y=30), TwistAxis="1,0,0")) -> (Output_Get:double)
   n8 = Get Skel_Mesh()
   > Set thigh_rot_inr_l(thigh_rot_inr_l:double=GetAlphaFromReferenceTransformDelta(SkeletalMeshComponent=@n8, BoneName=thigh_l, BaseName=pelvis, DeltaComponentType=EulerY, InputRange=(X=10,Y=-6), TwistAxis="1,0,0")) -> (Output_Get:double)
   > Set thigh_rot_inr_r(thigh_rot_inr_r:double=GetAlphaFromReferenceTransformDelta(SkeletalMeshComponent=@n8, BoneName=thigh_r, BaseName=pelvis, DeltaComponentType=EulerY, InputRange=(X=0,Y=-10), TwistAxis="1,0,0")) -> (Output_Get:double)
   > Set Arm R rot(Arm R rot:double=GetAlphaFromReferenceTransformDelta(SkeletalMeshComponent=Get Skel_Mesh(), BoneName=upperarm_r, BaseName=spine_05, InputRange=(X=60,Y=120), TwistAxis="1,0,0")) -> (Output_Get:double)
   > n9 = Branch(Condition=Get bIsCrouching())
   > Set thigh_shape_lt_l(thigh_shape_lt_l:double=0) -> (Output_Get:double)
   > n10 = Branch(Condition=Get bIsMoving2D())
   > Set crouch move nosim(crouch move nosim:bool=true) -> (Output_Get:bool)
   Set crouch move nosim(execute=[@n10.else, @n11.then], crouch move nosim=false)
   n11 = Set thigh_shape_lt_l(
    execute=@n9.else
    thigh_shape_lt_l=GetAlphaFromReferenceTransformDelta(
     SkeletalMeshComponent=Get Skel_Mesh()
     BoneName=calf_l
     BaseName=pelvis
     DeltaComponentType=EulerY
     InputRange=(X=0,Y=-60)
     TwistAxis="0,0,1"
    )
   )
   n12 = Cast To FortPlayerAnimInstance(execute=@n2.then_4, Object=GetAnimInstance(self=GetOwnerSkeletalMeshComponent()))
   > Set melee(melee:bool=false) -> (Output_Get:bool)
   > Set melee_crouch(melee_crouch:bool=false) -> (Output_Get:bool)
   > n13 = Switch on EFortWeaponCoreAnimation!(execute, Selection:EFortWeaponCoreAnimation Enum=Get WeaponCoreAnim(self=@n12)) -> (
    Melee:Exec
    Pistol:Exec
    Shotgun:Exec
    PaperBlueprint:Exec
    Rifle:Exec
    MeleeOneHand:Exec
    MachinePistol:Exec
    RocketLauncher:Exec
    GrenadeLauncher:Exec
    GoingCommando:Exec
    AssaultRifle:Exec
    TacticalShotgun:Exec
    SniperRifle:Exec
    TrapPlacement:Exec
    ShoulderLauncher:Exec
    AbilityDecoTool:Exec
    Crossbow:Exec
    C4:Exec
    RemoteControl:Exec
    DualWield:Exec
    AR_BullPup:Exec
    AR_ForwardGrip:Exec
    MedPackPaddles:Exec
    SMG_P90:Exec
    AR_DrumGun:Exec
    Consumable_Small:Exec
    Consumable_Large:Exec
    Balloon:Exec
    MountedTurret:Exec
    CreativeTool:Exec
    ExplosiveBow:Exec
    AshtonIndigo:Exec
    AshtonChicago:Exec
    MeleeDualWield:Exec
    Unarmed:Exec
   )
   Set melee(execute=[@n13.Melee, @n13.MeleeDualWield], melee=true)
   > n14 = Branch(Condition=Get bIsCrouching())
   > Set melee_crouch(melee_crouch=true)
   Set melee_crouch(execute=@n14.else, melee_crouch=false)
   n15 = Branch(execute=@n2.then_5, Condition=Get bIsInFrontEnd())
   > Set BwdMovement(BwdMovement:double=0) -> (Output_Get:double)
   > Set RightMovement(RightMovement:float=0) -> (Output_Get:float)
   > Set LeftMovement(LeftMovement:float=0) -> (Output_Get:float)
   n16 = PlayerMovement(execute=@n15.else, AnimInstance=Self-Reference(), XDotProductCoefficient=1, YDotProductCoefficient=1, ZDotProductCoefficient=1)
   n17 = MapRangeClamped(Value=@n16.YMovement, InRangeA=0, InRangeB=-1)
   n18 = float * float(A="Absolute (Float)"(A=@n16), B=0.5)
   > Set BwdMovement(BwdMovement=MapRangeClamped(
     Value=float + float(A=@n18, B=@n17) {is=title K2Node_CommutativeAssociativeBinaryOperator}
     InRangeA=0
     InRangeB=1.1
     OutRangeA=0.7
     OutRangeB=0
    ))
   n19 = MapRangeClamped(Value=@n16, InRangeA=-0.2, InRangeB=1)
   n20 = float * float(A="Absolute (Float)"(A=@n16.YMovement), B=0.2)
   > Set RightMovement(RightMovement=MapRangeClamped(Value=float + float(A=@n20, B=@n19), InRangeA=0, InRangeB=1.2, OutRangeB=0.7))
   n21 = MapRangeClamped(Value=@n16, InRangeA=0.2, InRangeB=-1, OutRangeB=0.3)
   n22 = float * float(A="Absolute (Float)"(A=@n16.YMovement), B=0.2)
   > Set LeftMovement(LeftMovement=MapRangeClamped(Value=float + float(A=@n22, B=@n21), InRangeA=0, InRangeB=1.2, OutRangeB=0.7))
   > Set BwdMovement_2(BwdMovement_2:double=MapRangeClamped(Value=@n16.YMovement, InRangeA=0, InRangeB=-1)) -> (Output_Get:double)
   n23 = Branch(execute=@n2.then_6, Condition=Get bIsSkyDiving())
   > n24 = Branch(Condition="float > float"(A=Get PawnVelocityZ(), B=50) {is=title K2Node_CallFunction})
   > Set BwdMovement(BwdMovement=float * float(A=Get BwdMovement(), B=0))
   Set BwdMovement(execute=[@n24.else, @n23.else], BwdMovement=float * float(A=Get BwdMovement(), B=1))
   n25 = GetAnimInstance(self=GetOwnerSkeletalMeshComponent())
   Is Valid(exec=@n2.then_7, InputObject=@n25)
   > n26 = Branch(Condition=OR Boolean!(
     A=Montage_IsPlaying(self=@n25, Montage=/BRCosmetics/Animation/Game/MainPlayer/Emotes/Emote_Breakdance_M1)
     B=Montage_IsPlaying(self=@n25, Montage=/BRCosmetics/Animation/Game/MainPlayer/Emotes/Breakdance2/Emote_Break_Dance2_CMM_M)
     C:bool=Montage_IsPlaying(self=@n25, Montage=/BRCosmetics/Animation/Game/MainPlayer/Emotes/Temple/Emote_Temple_CMM_M)
     D:bool=Montage_IsPlaying(self=@n25, Montage=/BRCosmetics/Animation/Game/MainPlayer/Emotes/SpectacleWeb/Emote_SpectacleWeb_CMM_M)
    ) -> (ReturnValue))
   > Set Emote(Emote:double=0) -> (Output_Get:double)
   > Set Emote_bool(Emote_bool:bool=true) -> (Output_Get:bool)
   Set Emote(execute=@n26.else, Emote=1)
   > Set Emote_bool(Emote_bool=false)
   n27 = Event OnBeginFalling()
   Set jumping(execute=[@n27.then, @n28.then], jumping:double=1) -> (Output_Get:double)
   n28 = Event OnBeginJumping()
   Event BlueprintInitializeAnimation()
   > Set Skel_Mesh(Skel_Mesh=GetOwnerSkeletalMeshComponent())
   n29 = Event OnEndFalling()
   Set jumping(execute=[@n29.then, @n30.then], jumping=0)
   n30 = Event OnEndJumping()
   Branch(Condition=Get bIsPlayingEmote())
   > Set GravityOverride(GravityOverride="0, 0,-480")
   Branch(Condition=Get bClothEnabled())
   > n31 = Cloth Controller Function DEPRECATED!(
    execute
    CharacterPartAnimInstance:Custom Character Part Anim Instance Object Reference=Self-Reference()
    ClothingAssetName:String=Gimmick_Cloth
    Cloth Paramaters:Cloth Params Struct Structure=Get Cloth Paramaters_0() -> (Cloth Paramaters_0:Cloth Params Struct Structure)
    Cloth Core Settings:Cloth Core Setting Struct Structure=Get Cloth Core Settings_0() -> (Cloth Core Settings_0:Cloth Core Setting Struct Structure)
    ClothUpAxis:Vector="1,0,0"
    OnGround_Moving_SpeedLimit:double=2
    OnGround_Moving_AngularVel_Max:double=0.05
    OnGround_Moving_FictitiousVel_Max:double=0.6
    LinVelScale_FwdBwd_Clamp:double=0
    LinVelScale_UpDwn_Clamp:double=0.7
    DeltaTime:double=Get DeltaTimeLocal() -> (DeltaTimeLocal:double)
    GameTimeInSeconds:double=GetGameTimeInSeconds()
    WindGustsParams:Wind Gust Params Struct Structure=Get Wind Gust Params() -> (Wind Gust Params:Wind Gust Params Struct Structure)
    WindGustLocalSettings:Wind Gust Local Setting Struct Structure=Get Gust Local Settings() -> (Gust Local Settings:Wind Gust Local Setting Struct Structure)
   ) -> (
    then
    CurrentStateOut:String
    Cloth Settings Out:Cloth Setting Struct Structure
    Cloth Core Settings Out:Cloth Core Setting Struct Structure
    WindGustSettings Out:Wind Gust Setting Struct Structure
    WindGustLocalSettings Out:Wind Gust Local Setting Struct Structure
    GustStrength - WindVelocityOut:Vector
    LinearVelVector-DELETE LATER:Vector
   )
   > Set Gust Local Settings(Gust Local Settings:Wind Gust Local Setting Struct Structure=@n31.WindGustLocalSettings Out) -> (Output_Get:Wind Gust Local Setting Struct Structure)
   Get melee() -> (melee:bool)
  ]}
]
variables [
 Arm R rot {type=number}
 BwdMovement {type=number}
 BwdMovement_2 {type=number}
 Cloth Core Settings_0 {
  type=ClothCoreSettingStruct
  default {
   collisionThickness=1.5
   dampingCoefficient=0.1
   frictionCoefficient=0.5
   localDampingCoefficient=0.1
   numberIterations=2
   numberSubsteps=1
   selfCollisionThickness=0
   animDriveDamping=<0 1>
   animDriveStiffness=<0 1>
   areaStiffness=<1 1>
   bendingStiffness=<1 1>
   edgeStiffness=<1 1>
   tetherStiffness=<1 1>
  }
 }
 Cloth Paramaters_0 {
  type=ClothParamsStruct
  default {
   dBNO {
    angularVelocityScale=0.6
    fictitiousAngularScale=0.1
    jointName=pelvis
    dragCoefficient=<0.2 1>
    gravityOverride=<-980 400 0>
    liftCoefficient=<0 0.5>
    linearVelocityScale=<0.3 0.3 0.1>
   }
   dBNO_Carried {
    angularVelocityScale=0.6
    fictitiousAngularScale=0.1
    jointName=pelvis
    dragCoefficient=<0.2 1>
    gravityOverride=<-980 -200 0>
    liftCoefficient=<0 0.5>
    linearVelocityScale=<0.3 0.3 0.1>
   }
   dBNO_Impulse {
    angularVelocityScale=0
    fictitiousAngularScale=0
    jointName=pelvis
    dragCoefficient=<0 0.2>
    gravityOverride=<-3000 -2000 0>
    liftCoefficient=<0 0.1>
    linearVelocityScale=<0.3 0.1 0.3>
   }
   dBNO_RollBwd {
    angularVelocityScale=0
    fictitiousAngularScale=0
    jointName=pelvis
    dragCoefficient=<0 0.2>
    gravityOverride=<-2000 0 0>
    liftCoefficient=<0 0.1>
    linearVelocityScale=<0.3 0.1 0.3>
   }
   dBNO_RollFwd {
    angularVelocityScale=0
    fictitiousAngularScale=0
    jointName=pelvis
    dragCoefficient=<0 0.2>
    gravityOverride=<-980 0 0>
    liftCoefficient=<0 0.1>
    linearVelocityScale=<0.3 0.1 0.3>
   }
   dBNO_Tumbling {
    angularVelocityScale=0
    fictitiousAngularScale=0
    jointName=pelvis
    dragCoefficient=<0 0.2>
    gravityOverride=<-2500 -300 0>
    liftCoefficient=<0 0.1>
    linearVelocityScale=<0.3 0.1 0.3>
   }
   emoteOrMelee {
    angularVelocityScale=0.6
    fictitiousAngularScale=0.1
    jointName=root
    dragCoefficient=<2.5 2.5>
    gravityOverride=<0 0 -980>
    liftCoefficient=<0.5 0.5>
    linearVelocityScale=<0.6 0.6 0.5>
   }
   front_End {
    angularVelocityScale=0.05
    fictitiousAngularScale=0.6
    jointName=root
    dragCoefficient=<0.2 0.2>
    gravityOverride=<0 -200 -980>
    liftCoefficient=<0.1 0.1>
    linearVelocityScale=<0.1 0.1 0.1>
   }
   onGround_Crouch_Moving {
    angularVelocityScale=0.1
    fictitiousAngularScale=0.3
    jointName=root
    dragCoefficient=<2 2>
    gravityOverride=<0 0 -980>
    liftCoefficient=<0.5 0.5>
    linearVelocityScale=<0.6 0.6 0.5>
   }
   onGround_Falling {
    angularVelocityScale=0.6
    fictitiousAngularScale=0
    jointName=root
    dragCoefficient=<0 0.2>
    gravityOverride=<0 -500 -980>
    liftCoefficient=<0 0.1>
    linearVelocityScale=<2 1 0.3>
   }
   onGround_Moving {
    angularVelocityScale=0.6
    fictitiousAngularScale=0.1
    jointName=root
    dragCoefficient=<2 2>
    gravityOverride=<0 0 -980>
    liftCoefficient=<0.5 0.5>
    linearVelocityScale=<0.6 0.6 0.5>
   }
   onGround_Standing {
    angularVelocityScale=0.05
    fictitiousAngularScale=1.2
    jointName=root
    dragCoefficient=<2.5 2.5>
    gravityOverride=<0 0 -980>
    liftCoefficient=<0.5 0.5>
    linearVelocityScale=<0.6 0.6 0.5>
   }
   rail_Grinding {
    angularVelocityScale=0.2
    fictitiousAngularScale=0
    jointName=root
    dragCoefficient=<0.1 0.2>
    gravityOverride=<50 -500 -400>
    liftCoefficient=<0 0.5>
    linearVelocityScale=<0.1 0.1 0.1>
   }
   riding_Moving {
    angularVelocityScale=0.05
    fictitiousAngularScale=1
    jointName=root
    dragCoefficient=<0.5 0.5>
    gravityOverride=<0 0 -980>
    liftCoefficient=<0.5 0.5>
    linearVelocityScale=<0.4 0.6 0.3>
   }
   riding_Moving_Targeting {
    angularVelocityScale=0.05
    fictitiousAngularScale=0.7
    jointName=root
    dragCoefficient=<0.5 0.5>
    gravityOverride=<0 0 -980>
    liftCoefficient=<0.5 0.5>
    linearVelocityScale=<0.2 0.6 0.2>
   }
   riding_Standing {
    angularVelocityScale=0.05
    fictitiousAngularScale=1
    jointName=root
    dragCoefficient=<0.5 0.5>
    gravityOverride=<0 -450 -980>
    liftCoefficient=<0.5 0.5>
    linearVelocityScale=<0.6 0.6 0.3>
   }
   skydive_DivingDown {
    angularVelocityScale=0.1
    fictitiousAngularScale=0
    jointName=pelvis
    dragCoefficient=<0.3 0.3>
    gravityOverride=<-450 -100 0>
    liftCoefficient=<0.9 0.9>
    linearVelocityScale=<0.1 0.1 0.1>
   }
   skydive_FloatingDown {
    angularVelocityScale=0.1
    fictitiousAngularScale=0
    jointName=root
    dragCoefficient=<0.3 0.3>
    gravityOverride=<0 -800 200>
    liftCoefficient=<1.3 1.3>
    linearVelocityScale=<0.2 0.2 0.1>
   }
   skydive_GoingUp {
    angularVelocityScale=0.1
    fictitiousAngularScale=0
    jointName=root
    dragCoefficient=<0.01 0.01>
    gravityOverride=<0 -2000 400>
    liftCoefficient=<0.05 0.05>
    linearVelocityScale=<0.1 0.1 0.01>
   }
   skydive_Parachuting {
    angularVelocityScale=0.6
    fictitiousAngularScale=0.1
    jointName=root
    dragCoefficient=<0.8 0.8>
    gravityOverride=<0 -600 -500>
    liftCoefficient=<0.8 0.8>
    linearVelocityScale=<0.2 0.2 0.3>
   }
   skydive_Surfing {
    angularVelocityScale=0.6
    fictitiousAngularScale=0.1
    jointName=root
    dragCoefficient=<1.5 1.5>
    gravityOverride=<0 -600 -500>
    liftCoefficient=<1 1>
    linearVelocityScale=<0.1 0.1 0.3>
   }
   swimming {
    angularVelocityScale=0.6
    fictitiousAngularScale=0
    jointName=root
    dragCoefficient=<2.5 2.5>
    gravityOverride=<0 0 0>
    liftCoefficient=<0.1 0.1>
    linearVelocityScale=<0.6 0.6 0.5>
   }
   tactical_Sprint {
    angularVelocityScale=0.6
    fictitiousAngularScale=0.1
    jointName=root
    dragCoefficient=<2 2>
    gravityOverride=<0 0 -980>
    liftCoefficient=<0.5 0.5>
    linearVelocityScale=<0.6 0.6 0.5>
   }
   ziplining {
    angularVelocityScale=0.6
    fictitiousAngularScale=0
    jointName=root
    dragCoefficient=<0 1.5>
    gravityOverride=<0 0 -980>
    liftCoefficient=<0 0.3>
    linearVelocityScale=<0.4 0.4 0.5>
   }
  }
 }
 crouch move nosim {type=boolean}
 DeltaTimeLocal {type=number}
 Emote {type=number}
 Emote_bool {type=boolean}
 GravityOverride {type=Vector}
 GravityOverrideParams {type=GravityOverrideParamsStruct, default {emoteOrMelee {gravityOverride=<0 0 -480>}}}
 Gust Local Settings {type=WindGustLocalSettingStruct, default {elapsedTime=0, finalGustStrengthLocal=0, gustDuration=0, gustStrength=0, timeBetweenGusts=0}}
 jumping {type=number}
 LeftMovement {type=number}
 melee {type=boolean}
 melee_crouch {type=boolean}
 RightMovement {type=number}
 Skel_Mesh {type=SkeletalMeshComponent}
 "skydive " {type=boolean}
 still_crouch {type=boolean}
 thigh_rot_inr_l {type=number}
 thigh_rot_inr_r {type=number}
 thigh_rot_otr_l {type=number}
 thigh_rot_otr_r {type=number}
 thigh_shape_lt_l {type=number}
 Wind Gust Params {
  type=WindGustParamsStruct
  default {
   dBNO {
    maxDownInterpSpeed=5
    maxGustDuration=2
    maxGustStrength=7
    maxTimeBetweenGusts=7
    maxUpInterpSpeed=5
    minDownInterpSpeed=2
    minGustDuration=0.5
    minGustStrength=3
    minTimeBetweenGusts=0
    minUpInterpSpeed=2
    windTurbulenceScalarA=0.1
    windTurbulenceScalarB=1
    windAmplitude=<0 0 0>
    windFrequency=<0 0 0>
    windOffset=<0 0 0>
   }
   dBNO_Carried {
    maxDownInterpSpeed=5
    maxGustDuration=2
    maxGustStrength=7
    maxTimeBetweenGusts=7
    maxUpInterpSpeed=5
    minDownInterpSpeed=2
    minGustDuration=0.5
    minGustStrength=3
    minTimeBetweenGusts=0
    minUpInterpSpeed=2
    windTurbulenceScalarA=0.1
    windTurbulenceScalarB=1
    windAmplitude=<0 0 0>
    windFrequency=<0 0 0>
    windOffset=<0 0 0>
   }
   dBNO_Impulse {
    maxDownInterpSpeed=5
    maxGustDuration=2
    maxGustStrength=7
    maxTimeBetweenGusts=7
    maxUpInterpSpeed=5
    minDownInterpSpeed=2
    minGustDuration=0.5
    minGustStrength=3
    minTimeBetweenGusts=0
    minUpInterpSpeed=2
    windTurbulenceScalarA=0.1
    windTurbulenceScalarB=1
    windAmplitude=<0 0 0>
    windFrequency=<0 0 0>
    windOffset=<0 0 0>
   }
   dBNO_RollBwd {
    maxDownInterpSpeed=5
    maxGustDuration=2
    maxGustStrength=7
    maxTimeBetweenGusts=7
    maxUpInterpSpeed=5
    minDownInterpSpeed=2
    minGustDuration=0.5
    minGustStrength=3
    minTimeBetweenGusts=0
    minUpInterpSpeed=2
    windTurbulenceScalarA=0.1
    windTurbulenceScalarB=1
    windAmplitude=<0 0 0>
    windFrequency=<0 0 0>
    windOffset=<0 0 0>
   }
   dBNO_RollFwd {
    maxDownInterpSpeed=5
    maxGustDuration=2
    maxGustStrength=7
    maxTimeBetweenGusts=7
    maxUpInterpSpeed=5
    minDownInterpSpeed=2
    minGustDuration=0.5
    minGustStrength=3
    minTimeBetweenGusts=0
    minUpInterpSpeed=2
    windTurbulenceScalarA=0.1
    windTurbulenceScalarB=1
    windAmplitude=<0 0 0>
    windFrequency=<0 0 0>
    windOffset=<0 0 0>
   }
   dBNO_Tumbling {
    maxDownInterpSpeed=5
    maxGustDuration=2
    maxGustStrength=7
    maxTimeBetweenGusts=7
    maxUpInterpSpeed=5
    minDownInterpSpeed=2
    minGustDuration=0.5
    minGustStrength=3
    minTimeBetweenGusts=0
    minUpInterpSpeed=2
    windTurbulenceScalarA=0.1
    windTurbulenceScalarB=1
    windAmplitude=<0 0 0>
    windFrequency=<0 0 0>
    windOffset=<0 0 0>
   }
   emoteOrMelee {
    maxDownInterpSpeed=5
    maxGustDuration=2
    maxGustStrength=7
    maxTimeBetweenGusts=7
    maxUpInterpSpeed=5
    minDownInterpSpeed=2
    minGustDuration=0.5
    minGustStrength=3
    minTimeBetweenGusts=0
    minUpInterpSpeed=2
    windTurbulenceScalarA=0.1
    windTurbulenceScalarB=1
    windAmplitude=<0 0 0>
    windFrequency=<0 0 0>
    windOffset=<0 0 0>
   }
   front_End {
    maxDownInterpSpeed=5
    maxGustDuration=2
    maxGustStrength=7
    maxTimeBetweenGusts=7
    maxUpInterpSpeed=5
    minDownInterpSpeed=2
    minGustDuration=0.5
    minGustStrength=3
    minTimeBetweenGusts=0
    minUpInterpSpeed=2
    windTurbulenceScalarA=0.1
    windTurbulenceScalarB=1
    windAmplitude=<0 0 0>
    windFrequency=<0 0 0>
    windOffset=<0 0 0>
   }
   onGround_Crouch_Moving {
    maxDownInterpSpeed=5
    maxGustDuration=2
    maxGustStrength=7
    maxTimeBetweenGusts=7
    maxUpInterpSpeed=5
    minDownInterpSpeed=2
    minGustDuration=0.5
    minGustStrength=3
    minTimeBetweenGusts=0
    minUpInterpSpeed=2
    windTurbulenceScalarA=0.1
    windTurbulenceScalarB=1
    windAmplitude=<0 0 0>
    windFrequency=<0 0 0>
    windOffset=<0 0 0>
   }
   onGround_Falling {
    maxDownInterpSpeed=5
    maxGustDuration=2
    maxGustStrength=7
    maxTimeBetweenGusts=7
    maxUpInterpSpeed=5
    minDownInterpSpeed=2
    minGustDuration=0.5
    minGustStrength=3
    minTimeBetweenGusts=0
    minUpInterpSpeed=2
    windTurbulenceScalarA=0.1
    windTurbulenceScalarB=1
    windAmplitude=<0 0 0>
    windFrequency=<0 0 0>
    windOffset=<0 0 0>
   }
   onGround_Moving {
    maxDownInterpSpeed=5
    maxGustDuration=2
    maxGustStrength=7
    maxTimeBetweenGusts=7
    maxUpInterpSpeed=5
    minDownInterpSpeed=2
    minGustDuration=0.5
    minGustStrength=3
    minTimeBetweenGusts=0
    minUpInterpSpeed=2
    windTurbulenceScalarA=0.1
    windTurbulenceScalarB=1
    windAmplitude=<0 0 0>
    windFrequency=<0 0 0>
    windOffset=<0 0 0>
   }
   onGround_Standing {
    maxDownInterpSpeed=5
    maxGustDuration=2
    maxGustStrength=7
    maxTimeBetweenGusts=7
    maxUpInterpSpeed=5
    minDownInterpSpeed=2
    minGustDuration=0.5
    minGustStrength=3
    minTimeBetweenGusts=0
    minUpInterpSpeed=2
    windTurbulenceScalarA=0.1
    windTurbulenceScalarB=1
    windAmplitude=<0 0 0>
    windFrequency=<0 0 0>
    windOffset=<0 0 0>
   }
   rail_Grinding {
    maxDownInterpSpeed=5
    maxGustDuration=8
    maxGustStrength=7
    maxTimeBetweenGusts=8
    maxUpInterpSpeed=5
    minDownInterpSpeed=2
    minGustDuration=0.2
    minGustStrength=3
    minTimeBetweenGusts=1
    minUpInterpSpeed=2
    windTurbulenceScalarA=0.1
    windTurbulenceScalarB=0.8
    windAmplitude=<20 65 42>
    windFrequency=<5 6 8.6>
    windOffset=<50 -500 -400>
   }
   riding_Moving {
    maxDownInterpSpeed=5
    maxGustDuration=2
    maxGustStrength=7
    maxTimeBetweenGusts=7
    maxUpInterpSpeed=5
    minDownInterpSpeed=2
    minGustDuration=0.5
    minGustStrength=3
    minTimeBetweenGusts=0
    minUpInterpSpeed=2
    windTurbulenceScalarA=0.1
    windTurbulenceScalarB=1
    windAmplitude=<0 0 0>
    windFrequency=<0 0 0>
    windOffset=<0 0 0>
   }
   riding_Moving_Targeting {
    maxDownInterpSpeed=5
    maxGustDuration=2
    maxGustStrength=7
    maxTimeBetweenGusts=7
    maxUpInterpSpeed=5
    minDownInterpSpeed=2
    minGustDuration=0.5
    minGustStrength=3
    minTimeBetweenGusts=0
    minUpInterpSpeed=2
    windTurbulenceScalarA=0.1
    windTurbulenceScalarB=1
    windAmplitude=<0 0 0>
    windFrequency=<0 0 0>
    windOffset=<0 0 0>
   }
   riding_Standing {
    maxDownInterpSpeed=5
    maxGustDuration=2
    maxGustStrength=7
    maxTimeBetweenGusts=7
    maxUpInterpSpeed=5
    minDownInterpSpeed=2
    minGustDuration=0.5
    minGustStrength=3
    minTimeBetweenGusts=0
    minUpInterpSpeed=2
    windTurbulenceScalarA=0.1
    windTurbulenceScalarB=1
    windAmplitude=<0 0 0>
    windFrequency=<0 0 0>
    windOffset=<0 0 0>
   }
   skydive_DivingDown {
    maxDownInterpSpeed=5
    maxGustDuration=2
    maxGustStrength=7
    maxTimeBetweenGusts=7
    maxUpInterpSpeed=5
    minDownInterpSpeed=2
    minGustDuration=0.5
    minGustStrength=3
    minTimeBetweenGusts=0
    minUpInterpSpeed=2
    windTurbulenceScalarA=0.1
    windTurbulenceScalarB=1
    windAmplitude=<0 0 0>
    windFrequency=<0 0 0>
    windOffset=<0 0 0>
   }
   skydive_FloatingDown {
    maxDownInterpSpeed=5
    maxGustDuration=2
    maxGustStrength=7
    maxTimeBetweenGusts=7
    maxUpInterpSpeed=5
    minDownInterpSpeed=2
    minGustDuration=0.5
    minGustStrength=3
    minTimeBetweenGusts=0
    minUpInterpSpeed=2
    windTurbulenceScalarA=0.1
    windTurbulenceScalarB=1
    windAmplitude=<0 0 0>
    windFrequency=<0 0 0>
    windOffset=<0 0 0>
   }
   skydive_GoingUp {
    maxDownInterpSpeed=5
    maxGustDuration=2
    maxGustStrength=7
    maxTimeBetweenGusts=7
    maxUpInterpSpeed=5
    minDownInterpSpeed=2
    minGustDuration=0.5
    minGustStrength=3
    minTimeBetweenGusts=0
    minUpInterpSpeed=2
    windTurbulenceScalarA=0.1
    windTurbulenceScalarB=1
    windAmplitude=<0 0 0>
    windFrequency=<0 0 0>
    windOffset=<0 0 0>
   }
   skydive_Parachuting {
    maxDownInterpSpeed=5
    maxGustDuration=2
    maxGustStrength=7
    maxTimeBetweenGusts=7
    maxUpInterpSpeed=5
    minDownInterpSpeed=2
    minGustDuration=0.5
    minGustStrength=3
    minTimeBetweenGusts=0
    minUpInterpSpeed=2
    windTurbulenceScalarA=0.1
    windTurbulenceScalarB=1
    windAmplitude=<0 0 0>
    windFrequency=<0 0 0>
    windOffset=<0 0 0>
   }
   skydive_Surfing {
    maxDownInterpSpeed=5
    maxGustDuration=2
    maxGustStrength=7
    maxTimeBetweenGusts=7
    maxUpInterpSpeed=5
    minDownInterpSpeed=2
    minGustDuration=0.5
    minGustStrength=3
    minTimeBetweenGusts=0
    minUpInterpSpeed=2
    windTurbulenceScalarA=0.1
    windTurbulenceScalarB=1
    windAmplitude=<0 0 0>
    windFrequency=<0 0 0>
    windOffset=<0 0 0>
   }
   swimming {
    maxDownInterpSpeed=5
    maxGustDuration=2
    maxGustStrength=7
    maxTimeBetweenGusts=7
    maxUpInterpSpeed=5
    minDownInterpSpeed=2
    minGustDuration=0.5
    minGustStrength=3
    minTimeBetweenGusts=0
    minUpInterpSpeed=2
    windTurbulenceScalarA=0.1
    windTurbulenceScalarB=1
    windAmplitude=<0 0 0>
    windFrequency=<0 0 0>
    windOffset=<0 0 0>
   }
   tactical_Sprint {
    maxDownInterpSpeed=5
    maxGustDuration=2
    maxGustStrength=7
    maxTimeBetweenGusts=7
    maxUpInterpSpeed=5
    minDownInterpSpeed=2
    minGustDuration=0.5
    minGustStrength=3
    minTimeBetweenGusts=0
    minUpInterpSpeed=2
    windTurbulenceScalarA=0.1
    windTurbulenceScalarB=1
    windAmplitude=<0 0 0>
    windFrequency=<0 0 0>
    windOffset=<0 0 0>
   }
   ziplining {
    maxDownInterpSpeed=5
    maxGustDuration=2
    maxGustStrength=7
    maxTimeBetweenGusts=7
    maxUpInterpSpeed=5
    minDownInterpSpeed=2
    minGustDuration=0.5
    minGustStrength=3
    minTimeBetweenGusts=0
    minUpInterpSpeed=2
    windTurbulenceScalarA=0.1
    windTurbulenceScalarB=1
    windAmplitude=<0 0 0>
    windFrequency=<0 0 0>
    windOffset=<0 0 0>
   }
  }
 }
]
