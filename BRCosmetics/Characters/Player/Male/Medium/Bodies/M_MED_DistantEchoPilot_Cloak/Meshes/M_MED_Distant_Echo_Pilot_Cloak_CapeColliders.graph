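# Physics asset of cape colliders for the male medium Distant Echo Pilot cloak cosmetic, by its name a collision proxy for cape simulation.
# Overlapping-sized capsules on pelvis and three spine bones form the torso collider.
# A large sphere on a FloorCOL bone, centered well below its origin, approximates the ground.
# Thigh, calf, foot and lower-arm bodies carry no shapes, and no constraints link any bodies.
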PhysicsAsset /BRCosmetics/Characters/Player/Male/Medium/Bodies/M_MED_DistantEchoPilot_Cloak/Meshes/M_MED_Distant_Echo_Pilot_Cloak_CapeColliders
bodies {
 FloorCOL [{shapeType=Sphere, radius=200, shapeName=floorCOL_sphere, center=<-3.4923 -2.7141 -191.7929>}]
 calf_l []
 calf_r []
 foot_l []
 foot_r []
 lowerarm_l []
 lowerarm_r []
 pelvis [{shapeType=Capsule, length=10.1562, radius=11.9138, center=<-2.6215 -1.3695 0>}]
 spine_01 [{shapeType=Capsule, length=9.5937, radius=11.4138, center=<-2.6215 -0.7156 0>}]
 spine_02 [{shapeType=Capsule, length=8.8437, radius=10.57, center=<-2.6215 0.4993 0.0001>}]
 spine_03 [{shapeType=Capsule, length=9.1562, radius=10.7575, center=<-1.953 0.1364 0.0001>}]
 thigh_l []
 thigh_r []
}
constraints []
